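# Editor utility widget for engineering that finds subclasses and assets of a target class in the project or loaded level.
# Asset Registry queries or recursive building level-instance scans feed a sorted, text-filtered, paged list that can select level actors or sync the Content Browser.
# Visible results feed text export, size and property reports, and a hard-coded custom filter.
# These batch jobs step through items with intermittent delays and support cancelling.
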EditorUtilityWidgetBlueprint /Game/Blueprints/Blutilities/ForEngineering/FindClassUsages_EWidget
graphs [
 AddAssetPath {nodes [
   n1 = AddAssetPath!() -> (then, TopLevelAssetPath:"Top Level Asset Path Structure (by ref)", AssetClass:Actor Class Reference, LevelInstance:Actor Object Reference) {is=title K2Node_FunctionEntry}
   n2 = BreakTopLevelAssetPath!(TopLevelAssetPath:"Top Level Asset Path Structure (by ref)"=@n1) -> (PathString:String)
   > n3 = Branch(Condition=Contains(SearchIn=@n2, Substring=.SKEL_))
   Set AssetPathString(execute=@n3.else, AssetPathString:String=@n2) -> (Output_Get:String)
   > n4 = Branch(Condition=StartsWith(SourceString=@n2, InPrefix=/Script/))
   > n5 = Branch(Condition=Get bIncludeNativeClasses() -> (bIncludeNativeClasses:bool))
   Add Unique(
    execute=[@n5.then, @n4.else]
    TargetArray=Get AllAssets() -> (AllAssets:Array of Strings)
    NewItem=Get AssetPathString() -> (AssetPathString:String)
   )
   > Is Valid(InputObject=Get LevelInstance() -> (LevelInstance:Actor Object Reference))
   > n6 = Branch(Condition=Contains!(
     TargetMap:Map of Strings to Integers=Get AssetToInstanceIndexMap() -> (AssetToInstanceIndexMap:Map of Strings to Integers)
     Key:"String (by ref)"=Get AssetPathString()
    ) -> (ReturnValue))
   n7 = Find(TargetMap:Map of Strings to Integers=Get AssetToInstanceIndexMap(), Key=Get AssetPathString()) -> (Value:int)
   n8 = Get AssetInstances() -> (AssetInstances:Array of Find Class Usages Level Instance Structures)
   Add(
    execute=[@n10.then, @n6.then]
    TargetArray:Array of Actor Object References=Break Find Class Usages Level Instance!(FindClassUsages_LevelInstance:"Find Class Usages Level Instance Structure (by ref)"="Get (a ref)"(Array:Array of Find Class Usages Level Instance Structures=@n8, Dimension 1=@n7) -> (Output:"Find Class Usages Level Instance Structure (by ref)")) -> (
     Class_9_#1:Actor Class Reference
     Instances_8_#2:Array of Actor Object References
    ) {is=title K2Node_BreakStruct}.Instances_8_#2
    NewItem:"Actor Object Reference (by ref)"=Get LevelInstance()
   ) {is=title K2Node_CallArrayFunction}
   n9 = Add(
    execute=@n6.else
    TargetArray:Array of Find Class Usages Level Instance Structures=Get AssetInstances()
    NewItem:"Find Class Usages Level Instance Structure (by ref)"=Make Find Class Usages Level Instance!(
     Class_9_#1:Actor Class Reference=Get AssetClass() -> (AssetClass:Actor Class Reference)
     Instances_8_#2:Array of Actor Object References
    ) -> (FindClassUsages_LevelInstance:Find Class Usages Level Instance Structure) {is=title K2Node_MakeStruct}
   )
   > n10 = Add!(
    execute
    TargetMap:Map of Strings to Integers=Get AssetToInstanceIndexMap()
    Key:"String (by ref)"=Get AssetPathString()
    Value:"Integer (by ref)"=@n9
   ) -> (then) {is=title K2Node_CallFunction}
  ]}
 AddLevelActors {nodes [
   n1 = AddLevelActors!() -> (then, Actors:Array of Actor Object References, TargetClass:Object Class Reference) {is=title K2Node_FunctionEntry}
   > n2 = For Each Loop(Array:Array of Actor Object References=@n1) -> (Array Element:Actor Object Reference)
   > n3 = Cast To BuildingLevelInstance!(execute, Object=@n2) -> (then, CastFailed, AsBuilding Level Instance:Building Level Instance Object Reference) {is=title K2Node_DynamicCast}
   > n4 = GetActorsInLevel!(execute, self:Building Level Instance Object Reference=@n3) -> (then, ReturnValue:Array of Actor Object References)
   > AddLevelActors!(
    execute
    self
    Actors:Array of Actor Object References=@n4
    TargetClass:Object Class Reference=Get TargetClass() -> (TargetClass:Object Class Reference)
   ) -> (then) {is=title K2Node_CallFunction}
   n5 = Get Class(Object=@n2)
   Branch(execute=@n3.CastFailed, Condition=ClassIsChildOf(TestClass=@n5, ParentClass=Get TargetClass()))
   > AddAssetPath!(
    execute
    self
    TopLevelAssetPath:"Top Level Asset Path Structure (by ref)"="Get Class Path Name (Top Level Asset Path)"!(Class:Object Class Reference=@n5) -> (ReturnValue:Top Level Asset Path Structure) {is=title K2Node_CallFunction}
    AssetClass:Actor Class Reference=@n5
    LevelInstance:Actor Object Reference=@n2
   ) -> (then) {is=title K2Node_CallFunction}
  ]}
 Apply Custom Filter {nodes [
   n1 = Apply Custom Filter!() -> (then, ClassDefaultObject:Object Reference) {is=title K2Node_FunctionEntry}
   n2 = RunCustomFilterForFindClassUsages!(execute, Object:Object Reference=@n1) -> (then, ReturnValue:bool)
   n3 = Cast To FortWeaponModItemDefinition!(execute=@n1.then, Object=@n1) -> (then, CastFailed, AsFort Weapon Mod Item Definition:Fort Weapon Mod Item Definition Object Reference) {is=title K2Node_DynamicCast}
   Return Node!(
    execute=[@n2.then, @n3.then]
    bVisible:bool="Not Equal (Object)"(A:Object Reference=@n3, B:Object Reference=/WeaponMods/Items/WeaponMods/ModGE/GE_ModWeaponAttributes) -> (ReturnValue:bool=~) {is=title K2Node_PromotableOperator}
   )
   Return Node(execute=@n3.CastFailed, bVisible=false)
  ]}
 ApplyFilter {nodes [
   n1 = ApplyFilter!() -> (then, AssetPath:String, RequiredFIlters:Array of Strings, ExcludedFilters:Array of Strings) {is=title K2Node_FunctionEntry}
   > Set Label(Label:String=@n1) -> (Output_Get:String)
   > n2 = Branch(Condition=Is Empty!(TargetArray=@n1.RequiredFIlters) -> (ReturnValue:bool))
   > n3 = Set bMatchAnyRequired(bMatchAnyRequired:bool=true) -> (Output_Get:bool)
   n4 = For Each Loop(Exec=[@n6.Completed, @n3.then], Array:Array of Strings=@n1.ExcludedFilters) -> (Array Element:String)
   n5 = Get bMatchAnyExcluded() -> (bMatchAnyExcluded:bool)
   Return Node(execute=@n4.Completed, bVisible=AND Boolean(A=Get bMatchAnyRequired() -> (bMatchAnyRequired:bool), B=NOT Boolean(A=@n5)))
   > Branch(Condition=Contains!(SearchIn:String=Get Label() -> (Label:String), Substring:String=@n4, bUseCase:bool, bSearchFromEnd:bool) -> (ReturnValue))
   > Set bMatchAnyExcluded(bMatchAnyExcluded:bool=true) -> (Output_Get:bool)
   n6 = For Each Loop(Exec=@n2.else, Array=@n1.RequiredFIlters)
   > Branch(Condition=Contains(SearchIn=Get Label(), Substring=@n6))
   > Set bMatchAnyRequired(bMatchAnyRequired=true)
  ]}
 ApplyTextPadding {nodes [
   n1 = ApplyTextPadding!() -> (then, InStr:String) {is=title K2Node_FunctionEntry}
   > Return Node!(execute, OutStr:String=Append!(A=" ", B=@n1, C:String=" ") -> (ReturnValue))
  ]}
 CanExport {nodes [
   CanExport() {is=title K2Node_FunctionEntry}
   n1 = Break Directory Path!(DirectoryPath:"Directory Path Structure (by ref)"=Get ExportOutputFolder() -> (ExportOutputFolder:Directory Path Structure)) -> (Path:String) {is=title K2Node_BreakStruct}
   n2 = IsEmpty(InString=Trim!(SourceString:String=@n1) -> (ReturnValue:String))
   > Return Node!(execute, ReturnValue:bool=AND Boolean(A=NOT Boolean(A=@n2), B=DirectoryExists!(InPath:String=@n1) -> (ReturnValue:bool)))
  ]}
 ChangePage {nodes [
   ChangePage!() -> (then, bMoveRight:bool=~) {is=title K2Node_FunctionEntry}
   n1 = Length(TargetArray:Array of Strings=Get VisibleAssets() -> (VisibleAssets:Array of Strings))
   n2 = Get MaxItemsPerPage() -> (MaxItemsPerPage:int=0)
   > Set MaxPageIndexOffset(MaxPageIndexOffset:int=int * int(A=int / int(A=@n1, B=@n2), B=Get MaxItemsPerPage() -> (MaxItemsPerPage=0))) -> (Output_Get:int=0)
   > Set PreviousPageIndexOffset(PreviousPageIndexOffset:int=Get CurrentPageIndexOffset() -> (CurrentPageIndexOffset:int=0)) -> (Output_Get:int=0)
   n3 = int * int(A=Get MaxItemsPerPage() -> (MaxItemsPerPage=0), B=SelectInt(A=1, B=-1, bPickA=Get bMoveRight() -> (bMoveRight:bool)))
   n4 = Get CurrentPageIndexOffset() -> (CurrentPageIndexOffset=0)
   > n5 = Set CurrentPageIndexOffset(CurrentPageIndexOffset:int="Clamp (Integer)"(
     Value=int + int(A=@n4, B=@n3)
     Min=Get MinPageIndexOffset() -> (MinPageIndexOffset:int=0)
     Max=Get MaxPageIndexOffset() -> (MaxPageIndexOffset:int=0)
    ) -> (ReturnValue=0)) -> (Output_Get:int=0)
   > Branch(Condition="Not Equal (Integer)"(A=@n5, B=Get PreviousPageIndexOffset() -> (PreviousPageIndexOffset:int=0)))
   > UpdatePageItems()
  ]}
 CompareAssets {nodes [
   CompareAssets!() -> (then, LeftAsset:String, RightAsset:String) {is=title K2Node_FunctionEntry}
   > n1 = Branch(Condition=Get bUseLoadedLevel() -> (bUseLoadedLevel:bool))
   n2 = GetLevelInstanceCount!(Asset:String=Get RightAsset() -> (RightAsset:String)) -> (Count:int=0)
   n3 = GetLevelInstanceCount(Asset=Get LeftAsset() -> (LeftAsset:String)) -> (Count=0)
   > n4 = Branch(Condition="Not Equal (Integer)"(A=@n3, B=@n2))
   > Return Node(ReturnValue="integer > integer"(A=@n3, B=@n2))
   n5 = Get LeftAsset()
   n6 = Get RightAsset()
   Return Node(
    execute=[@n4.else, @n1.else]
    ReturnValue="integer < integer"(A=CompareStrings!(LeftStr:String=@n5, RightStr:String=@n6) -> (ReturnValue:int=0))
   )
  ]}
 CreateDefaultListItems {nodes [
   CreateDefaultListItems() {is=title K2Node_FunctionEntry}
   > clearItems(self:TA Generic 01 List Object Reference=Get LstAssetPaths() -> (LstAssetPaths:TA Generic 01 List Object Reference))
   > n1 = For Loop(LastIndex=Get MaxItemsPerPage() -> (MaxItemsPerPage=0))
   > n2 = addItemGeneric!(execute, self:TA Generic 01 List Object Reference=Get LstAssetPaths(), displayString:String="To String (Integer)"(InInt=@n1)) -> (then, widget:User Widget Object Reference, data:Object Reference)
   > n3 = Cast To TA_Generic01_ListItem!(execute, Object=@n2) -> (then, CastFailed, AsTA Generic 01 List Item:TA Generic 01 List Item Object Reference) {is=title K2Node_DynamicCast}
   > setTextAlignment!(execute, self:TA Generic 01 List Item Object Reference=@n3, alignment:EHorizontalAlignment Enum=HAlign_Left) -> (then)
   > SetVisibility(self=@n3, InVisibility=Collapsed)
  ]}
 DoNextCustomFilter {nodes [
   DoNextCustomFilter() {is=title K2Node_FunctionEntry}
   > n1 = Branch(Condition=Is Valid Index!(TargetArray=Get ProcessingAssets() -> (ProcessingAssets:Array of Strings), IndexToTest:int=Get ProcessingIndex() -> (ProcessingIndex:int=0)) -> (ReturnValue:bool))
   n2 = "Get (a ref)"(Array:Array of Strings=Get ProcessingAssets(), Dimension 1=Get ProcessingIndex() -> (ProcessingIndex=0)) -> (Output:"String (by ref)")
   > n3 = LoadAsset_Blocking(Asset=To Soft Object Reference(SoftObjectPath=MakeSoftObjectPath!(PathString:String=@n2) -> (ReturnValue:Soft Object Path Structure)))
   n4 = Get Class(Object=@n3) -> (ReturnValue:Object Class Reference)
   > n5 = Branch(Condition=OR Boolean(
     A=IsDataAssetClass!(Class:Object Class Reference=@n4) -> (ReturnValue:bool)
     B=IsDataTableClass!(Class:Object Class Reference=@n4) -> (ReturnValue:bool)
    ))
   > n6 = Apply Custom Filter!(execute, self, ClassDefaultObject:Object Reference=@n3) -> (then, bVisible:bool) {is=title K2Node_CallFunction}
   > n7 = Branch(Condition=@n6)
   n8 = Add(execute=[@n13.then, @n7.then], TargetArray:Array of Strings=Get VisibleAssets(), NewItem:"String (by ref)"=@n2) {is=title K2Node_CallArrayFunction}
   n9 = Increment Int(" "=[@n8.then, @n13.else, @n10.then, @n1.else, @n7.else], Value=Get ProcessingIndex() -> (ProcessingIndex=0))
   > UpdateProcessingStatusText!(execute, self, Index:int=@n9) -> (then)
   n10 = Branch(execute=@n5.else, Condition=StartsWith(SourceString=@n2, InPrefix=/Script/))
   n11 = GetClassDefaultObjectFromPath!(execute=@n10.else, ClassAssetPath:String=Split(SourceString=@n2, InStr=., SearchDir=FromStart)) -> (then, ReturnValue:Object Reference)
   > n12 = Apply Custom Filter(ClassDefaultObject=@n11)
   > n13 = Branch(Condition=@n12)
  ]}
 DoNextExport {nodes [
   DoNextExport() {is=title K2Node_FunctionEntry}
   > Branch(Condition=Is Valid Index(TargetArray=Get ProcessingAssets(), IndexToTest=Get ProcessingIndex() -> (ProcessingIndex=0)))
   > n1 = GetAssetTools()
   n2 = Get ProcessingIndex() -> (ProcessingIndex=0)
   n3 = Get ProcessingAssets()
   > ExportAssets!(
    execute
    self:Asset Tools Interface=@n1
    AssetsToExport:Array of Strings=Make Array("[0]":String="Get (a ref)"(Array=@n3, Dimension 1=@n2)) -> (Array:Array of Strings)
    ExportPath:String=Break Directory Path(DirectoryPath=Get ExportOutputFolder())
   ) -> (then)
   > n4 = Increment Int(Value=Get ProcessingIndex() -> (ProcessingIndex=0))
   > UpdateProcessingStatusText(Index=@n4)
  ]}
 DoNextPropertyReport {nodes [
   DoNextPropertyReport() {is=title K2Node_FunctionEntry}
   > Branch(Condition=Is Valid Index(TargetArray=Get ProcessingAssets(), IndexToTest=Get ProcessingIndex() -> (ProcessingIndex=0)))
   > UpdateClassPropertyReport!(
    execute
    InData:"Class Property Report Data Structure (by ref)"=Get PropertyReport() -> (PropertyReport:Class Property Report Data Structure)
    AssetPath:String="Get (a ref)"(Array=Get ProcessingAssets(), Dimension 1=Get ProcessingIndex() -> (ProcessingIndex=0))
   ) -> (then)
   > n1 = Increment Int(Value=Get ProcessingIndex() -> (ProcessingIndex=0))
   > UpdateProcessingStatusText(Index=@n1)
  ]}
 DoNextSizeReport {nodes [
   DoNextSizeReport() {is=title K2Node_FunctionEntry}
   > Branch(Condition=Is Valid Index(TargetArray=Get ProcessingAssets(), IndexToTest=Get ProcessingIndex() -> (ProcessingIndex=0)))
   > UpdateClassSizeReport!(
    execute
    InData:"Class Size Report Data Structure (by ref)"=Get SizeReport() -> (SizeReport:Class Size Report Data Structure)
    AssetPath:String="Get (a ref)"(Array=Get ProcessingAssets(), Dimension 1=Get ProcessingIndex() -> (ProcessingIndex=0))
   ) -> (then)
   > n1 = Increment Int(Value=Get ProcessingIndex() -> (ProcessingIndex=0))
   > UpdateProcessingStatusText(Index=@n1)
  ]}
 EventGraph {nodes [
   "On Clicked (btnAdvanced)"!() -> (OutputDelegate, then, widget:User Widget Object Reference, event:Pointer Event Structure) {is=title K2Node_ComponentBoundEvent}
   > ToggleAdvancedSelection()
   "On Property Changed (PropTargetClass)"!() -> (OutputDelegate, then, PropertyName:Name=None) {is=title K2Node_ComponentBoundEvent}
   > HandleAssetViewChanged!(execute, self, bClearActiveTargetClass:bool=false) -> (then)
   "On Property Changed (PropOnlyShowEditable)"!() -> (OutputDelegate, then, PropertyName:Name=None) {is=title K2Node_ComponentBoundEvent}
   > SetVisibility(
    self=[
     Get PropShowExposedInSpawn() -> (PropShowExposedInSpawn:Single Property View Object Reference)
     Get PropShowInstanceEditable() -> (PropShowInstanceEditable:Single Property View Object Reference)
     Get PropShowBPModifiable() -> (PropShowBPModifiable:Single Property View Object Reference)
    ]
    InVisibility=Select(Option 0=Collapsed, Option 1=Visible, Index=Get bOnlyShowEditable() -> (bOnlyShowEditable:bool))
   )
   "On Property Changed (PropCheckTimesDefaultChanges)"!() -> (OutputDelegate, then, PropertyName:Name=None) {is=title K2Node_ComponentBoundEvent}
   > SetVisibility(
    self=Get PropLogChangedProperties() -> (PropLogChangedProperties:Single Property View Object Reference)
    InVisibility=Select(Option 0=Collapsed, Option 1=Visible, Index=Get bCheckTimesDefaultChanged() -> (bCheckTimesDefaultChanged:bool))
   )
   "On Property Changed (PropIncludeNativeClasses)"!() -> (OutputDelegate, then, PropertyName:Name=None) {is=title K2Node_ComponentBoundEvent}
   > HandleAssetViewChanged(bClearActiveTargetClass=true)
   "On Clicked (btnPageLeft)"!() -> (OutputDelegate, then, widget:User Widget Object Reference, event:Pointer Event Structure) {is=title K2Node_ComponentBoundEvent}
   > ChangePage!(execute, self, bMoveRight:bool=false) -> (then) {is=title K2Node_CallFunction}
   "On Clicked (btnPageRight)"!() -> (OutputDelegate, then, widget:User Widget Object Reference, event:Pointer Event Structure) {is=title K2Node_ComponentBoundEvent}
   > ChangePage(bMoveRight=true)
   "On Clicked (btnApplyCustomFilter)"!() -> (OutputDelegate, then, widget:User Widget Object Reference, event:Pointer Event Structure) {is=title K2Node_ComponentBoundEvent}
   > n1 = Branch(Condition=Get bCustomFilterApplied() -> (bCustomFilterApplied:bool))
   > RunFilter()
   PrepareCustomFilter(execute=@n1.else)
   > SetProcessingWidgetStates!(execute, self, bProcessing:bool=true) -> (then)
   > n2 = UpdateProcessingStatusText(Index=0)
   n3 = Branch(execute=[@n2.then, @n6.then], Condition=IsProcessingFinished!() -> (ReturnValue:bool))
   > FinishCustomFilter()
   > SetProcessingWidgetStates(bProcessing=false)
   n4 = Should Delay Processing!(In:Exec=@n3.else) -> (DoDelay:Exec, DoNotDelay:Exec) {is=title K2Node_MacroInstance}
   > n5 = Delay!(execute, WorldContextObject:Object Reference, Duration=0.01) -> (then)
   n6 = DoNextCustomFilter(execute=[@n5.then, @n4.DoNotDelay]) {is=title K2Node_CallFunction}
   "On Clicked (btnOpenToolBP)"!() -> (OutputDelegate, then, widget:User Widget Object Reference, event:Pointer Event Structure) {is=title K2Node_ComponentBoundEvent}
   > n7 = GetPathNameForLoadedAsset(LoadedAsset=SelectObject!(
     A:Object Reference=&self
     B:Object Reference=&self
     bSelectA:bool=false
    ) -> (ReturnValue:Object Reference))
   > SyncBrowserToObjects(AssetPaths=Make Array("[0]"=@n7))
   n8 = "On Item Right Clicked (LstAssetPaths)"!() -> (OutputDelegate, then, isShiftDown:bool, isCtrlDown:bool, data:Object Reference, item:User Widget Object Reference) {is=title K2Node_ComponentBoundEvent}
   > n9 = Cast To TA_Generic01_ListItem(Object=@n8.item)
   > Branch(Condition=Get bUseLoadedLevel())
   > SelectLevelInstances!(
    execute
    self
    Asset:String=Split(SourceString=Get label!(self:TA Base Editor Button Widget Object Reference=@n9) -> (label:String), InStr=" - ", SearchDir=FromStart).RightS
   ) -> (then)
   "On Clicked (btnProjectAssets)"!() -> (OutputDelegate, then, widget:User Widget Object Reference, event:Pointer Event Structure) {is=title K2Node_ComponentBoundEvent}
   > SetAssetSource!(execute, self, bLoadedLevel:bool=false) -> (then)
   n10 = "On Property Changed (PropRequiredFilter)"!() -> (OutputDelegate, then, PropertyName:Name=None) {is=title K2Node_ComponentBoundEvent}
   RunFilter(execute=[@n10.then, @n11.then])
   "On Clicked (btnLoadedLevel)"!() -> (OutputDelegate, then, widget:User Widget Object Reference, event:Pointer Event Structure) {is=title K2Node_ComponentBoundEvent}
   > SetAssetSource(bLoadedLevel=true)
   n11 = "On Property Changed (PropExcludedFilter)"!() -> (OutputDelegate, then, PropertyName:Name=None) {is=title K2Node_ComponentBoundEvent}
   "On Clicked (btnExportToText)"!() -> (OutputDelegate, then, widget:User Widget Object Reference, event:Pointer Event Structure) {is=title K2Node_ComponentBoundEvent}
   > Branch(Condition=CanExport!() -> (ReturnValue:bool) {is=title K2Node_CallFunction})
   > PrepareProcessing()
   > SetProcessingWidgetStates(bProcessing=true)
   > n12 = UpdateProcessingStatusText(Index=0)
   n13 = Branch(execute=[@n12.then, @n16.then], Condition=IsProcessingFinished())
   > SetProcessingWidgetStates(bProcessing=false)
   n14 = Should Delay Processing(In=@n13.else)
   > n15 = Delay(Duration=0.01)
   n16 = DoNextExport(execute=[@n15.then, @n14.DoNotDelay]) {is=title K2Node_CallFunction}
   n17 = "On Item Left Clicked (LstAssetPaths)"!() -> (OutputDelegate, then, isShiftDown:bool, isCtrlDown:bool, data:Object Reference, item:User Widget Object Reference) {is=title K2Node_ComponentBoundEvent}
   > n18 = Cast To TA_Generic01_ListItem(Object=@n17.item)
   n19 = Get label(self=@n18)
   n20 = Split(SourceString=@n19, InStr=" - ", SearchDir=FromStart)
   n21 = SelectString(A=@n20.RightS, B=@n19, bPickA=@n20.ReturnValue)
   n22 = Split(SourceString=@n21, InStr=., SearchDir=FromStart)
   > SyncBrowserToObjects(AssetPaths=Make Array("[0]"=SelectString(A=@n22, B=@n21, bPickA=@n22.ReturnValue)))
   "On Clicked (btnGenerateSizeReport)"!() -> (OutputDelegate, then, widget:User Widget Object Reference, event:Pointer Event Structure) {is=title K2Node_ComponentBoundEvent}
   > n23 = PrepareSizeReport!(execute, self) -> (then, bValid:bool)
   > Branch(Condition=@n23)
   > SetProcessingWidgetStates(bProcessing=true)
   > n24 = UpdateProcessingStatusText(Index=0)
   n25 = Branch(execute=[@n24.then, @n28.then], Condition=IsProcessingFinished())
   > FinishSizeReport()
   > SetProcessingWidgetStates(bProcessing=false)
   n26 = Should Delay Processing(In=@n25.else)
   > n27 = Delay(Duration=0.01)
   n28 = DoNextSizeReport(execute=[@n27.then, @n26.DoNotDelay]) {is=title K2Node_CallFunction}
   "On Property Changed (PropTargetClassPath)"!() -> (OutputDelegate, then, PropertyName:Name=None) {is=title K2Node_ComponentBoundEvent}
   > HandleAssetViewChanged(bClearActiveTargetClass=true)
   n29 = "On Clicked (btnCancelExport)"!() -> (OutputDelegate, then, widget:User Widget Object Reference, event:Pointer Event Structure) {is=title K2Node_ComponentBoundEvent}
   Set ProcessingIndex(execute=[@n29.then, @n49.then], ProcessingIndex:int=99999999) -> (Output_Get:int=0)
   "On Clicked (btnGeneratePropertyReport)"!() -> (OutputDelegate, then, widget:User Widget Object Reference, event:Pointer Event Structure) {is=title K2Node_ComponentBoundEvent}
   > n30 = Prepare Property Report!(execute, self) -> (then, bValid:bool)
   > Branch(Condition=@n30)
   > SetProcessingWidgetStates(bProcessing=true)
   > UpdateProcessingStatusText(Index=0)
   > n31 = Branch(Condition=Get bCheckTimesDefaultChanged())
   n32 = Branch(execute=[@n31.then, @n35.then], Condition=IsProcessingFinished())
   FinishPropertyReport(execute=[@n32.then, @n31.else])
   > SetProcessingWidgetStates(bProcessing=false)
   n33 = Should Delay Processing(In=@n32.else)
   > n34 = Delay(Duration=0.01)
   n35 = DoNextPropertyReport(execute=[@n34.then, @n33.DoNotDelay]) {is=title K2Node_CallFunction}
   Event PreConstruct()
   n36 = Get PropTargetClass() -> (PropTargetClass:Single Property View Object Reference)
   > SetObject(self=@n36, NewObject=Self-Reference())
   > SetPropertyName!(execute, self:Single Property View Object Reference=@n36, NewPropertyName:Name=TargetClass) -> (then)
   n37 = Get PropRequiredFilter() -> (PropRequiredFilter:Single Property View Object Reference)
   > SetObject(self=@n37, NewObject=Self-Reference())
   > SetPropertyName(self=@n37, NewPropertyName=RequiredFilter)
   n38 = Get PropExcludedFilter() -> (PropExcludedFilter:Single Property View Object Reference)
   > SetObject(self=@n38, NewObject=Self-Reference())
   > SetPropertyName(self=@n38, NewPropertyName=ExcludedFilter)
   n39 = Get PropTargetClassPath() -> (PropTargetClassPath:Single Property View Object Reference)
   > SetObject(self=@n39, NewObject=Self-Reference())
   > SetPropertyName(self=@n39, NewPropertyName=TargetClassPath)
   n40 = Get PropOutputFolder() -> (PropOutputFolder:Single Property View Object Reference)
   > SetObject(self=@n40, NewObject=Self-Reference())
   > SetPropertyName(self=@n40, NewPropertyName=ExportOutputFolder)
   n41 = Get PropShowInstanceEditable()
   > SetObject(self=@n41, NewObject=Self-Reference())
   > SetPropertyName(self=@n41, NewPropertyName=bShowIntanceEditable)
   n42 = Get PropShowBPModifiable()
   > SetObject(self=@n42, NewObject=Self-Reference())
   > SetPropertyName(self=@n42, NewPropertyName=bShowBPModifiable)
   n43 = Get PropShowExposedInSpawn()
   > SetObject(self=@n43, NewObject=Self-Reference())
   > SetPropertyName(self=@n43, NewPropertyName=bShowExposedInSpawn)
   n44 = Get PropOnlyShowEditable() -> (PropOnlyShowEditable:Single Property View Object Reference)
   > SetObject(self=@n44, NewObject=Self-Reference())
   > SetPropertyName(self=@n44, NewPropertyName=bOnlyShowEditable)
   n45 = Get PropCheckTimesDefaultChanges() -> (PropCheckTimesDefaultChanges:Single Property View Object Reference)
   > SetObject(self=@n45, NewObject=Self-Reference())
   > SetPropertyName(self=@n45, NewPropertyName=bCheckTimesDefaultChanged)
   n46 = Get PropIncludeNativeClasses() -> (PropIncludeNativeClasses:Single Property View Object Reference)
   > SetObject(self=@n46, NewObject=Self-Reference())
   > SetPropertyName(self=@n46, NewPropertyName=bIncludeNativeClasses)
   n47 = Get PropLogChangedProperties()
   > SetObject(self=@n47, NewObject=Self-Reference())
   > SetPropertyName(self=@n47, NewPropertyName=LogChangedProperties)
   n48 = Get PropEditorClassSizeOffset() -> (PropEditorClassSizeOffset:Single Property View Object Reference)
   > SetObject(self=@n48, NewObject=Self-Reference())
   > SetPropertyName(self=@n48, NewPropertyName=EditorClassSizeOffset)
   Event Construct()
   > CreateDefaultListItems() {is=title K2Node_CallFunction}
   > UpdateAssetSourceWidgets()
   n49 = Event Destruct()
  ]}
 FinishCustomFilter {nodes [
   FinishCustomFilter() {is=title K2Node_FunctionEntry}
   > PrintString(InString=CUSTOM FILTER RESULTS...)
   > n1 = For Each Loop(Array=Get VisibleAssets())
   RefreshVisibleAssets(execute=@n1.Completed)
   > UpdateCustomFilterApplied!(execute, self, bIsApplied:bool=true) -> (then)
   PrintString(execute=@n1.LoopBody, InString=Append!(A="    ASSET:", B=@n1) -> (ReturnValue))
  ]}
 FinishPropertyReport {nodes [
   FinishPropertyReport() {is=title K2Node_FunctionEntry}
   > OutputClassPropertyReport!(execute, InData:"Class Property Report Data Structure (by ref)"=Get PropertyReport()) -> (then)
  ]}
 FinishSizeReport {nodes [
   FinishSizeReport() {is=title K2Node_FunctionEntry}
   > OutputClassSizeReport!(execute, InData:"Class Size Report Data Structure (by ref)"=Get SizeReport()) -> (then)
  ]}
 FinshedUpdatingAssets {nodes [
   FinshedUpdatingAssets() {is=title K2Node_FunctionEntry}
   > n1 = SortStringArrayWithPredicate!(execute, InStrings:Array of Strings=Get AllAssets(), Predicate:"Delegate (by ref)"=Create Event()) -> (then, OutSortedStrings:Array of Strings)
   > Set AllAssets(AllAssets:Array of Strings=@n1) -> (Output_Get:Array of Strings)
   > RunFilter()
   > UpdatePageItems()
  ]}
 GetAssetDisplayString {nodes [
   n1 = GetAssetDisplayString!() -> (then, Asset:String) {is=title K2Node_FunctionEntry}
   > n2 = Branch(Condition=Get bUseLoadedLevel())
   n3 = Find(TargetMap=Get AssetToInstanceIndexMap(), Key=@n1)
   n4 = Get AssetInstances()
   n5 = Length(TargetArray:Array of Actor Object References=Break Find Class Usages Level Instance(FindClassUsages_LevelInstance="Get (a ref)"(Array:Array of Find Class Usages Level Instance Structures=@n4, Dimension 1=@n3) -> (Output:"Find Class Usages Level Instance Structure (by ref)")).Instances_8_#2)
   > Return Node!(execute, DisplayString:String=Append!(A="To String (Integer)"(InInt=@n5), B=" - ", C:String=Get Asset() -> (Asset:String)) -> (ReturnValue))
   Return Node(execute=@n2.else, DisplayString=@n1)
  ]}
 GetDerivedClasses {nodes [
   n1 = GetDerivedClasses!() -> (then, ClassName:"Top Level Asset Path Structure (by ref)", ClassPath:String) {is=title K2Node_FunctionEntry}
   > n2 = GetDerivedClassNames!(
    execute
    self:Asset Registry Interface=GetAssetRegistry()
    ClassNames:Array of Top Level Asset Path Structures=Make Array("[0]":Top Level Asset Path Structure=@n1) -> (Array:Array of Top Level Asset Path Structures)
    ExcludedClassNames:Set of Top Level Asset Path Structures=Make Set!("[0]":Top Level Asset Path Structure) -> (Set:Set of Top Level Asset Path Structures) {is=title K2Node_MakeSet}
   ) -> (then, OutDerivedClassNames:Set of Top Level Asset Path Structures)
   > n3 = To Array!(execute, A:Set of Top Level Asset Path Structures=@n2) -> (then, Result:Array of Top Level Asset Path Structures)
   > Set ReturnArray(ReturnArray:Array of Top Level Asset Path Structures=@n3) -> (Output_Get:Array of Top Level Asset Path Structures)
   n4 = Get ClassPath() -> (ClassPath:String)
   > n5 = Branch(Condition=OR Boolean(A=IsDataAssetClassPath!(ClassPath:String=@n4) -> (bResult:bool), B=IsDataTableClassPath!(ClassPath:String=@n4) -> (bResult:bool)))
   > n6 = GetAssetsByClass!(
    execute
    self:Asset Registry Interface=GetAssetRegistry()
    ClassPathName:Top Level Asset Path Structure=Get ClassName() -> (ClassName:"Top Level Asset Path Structure (by ref)")
    bSearchSubClasses:bool=true
   ) -> (then, OutAssetData:Array of Asset Data Structures, ReturnValue:bool)
   > n7 = For Each Loop(Array:Array of Asset Data Structures=@n6) -> (Array Element:Asset Data Structure)
   Return Node!(
    execute=[@n7.Completed, @n5.else]
    Result:Array of Top Level Asset Path Structures=Get ReturnArray() -> (ReturnArray:Array of Top Level Asset Path Structures)
   )
   n8 = Break Asset Data(AssetData=@n7)
   > Add Unique(
    TargetArray:Array of Top Level Asset Path Structures=Get ReturnArray()
    NewItem:"Top Level Asset Path Structure (by ref)"=MakeTopLevelAssetPath!(PackageName:String="To String (Name)"(InName=@n8), AssetName:String="To String (Name)"(InName=@n8.AssetName)) -> (ReturnValue:Top Level Asset Path Structure)
   )
  ]}
 GetExcludedFilters {nodes [
   GetExcludedFilters() {is=title K2Node_FunctionEntry}
   > Return Node!(execute, ExcludedFilters:Array of Strings=ParseIntoArray(SourceString=Get ExcludedFilter() -> (ExcludedFilter:String), Delimiter=","))
  ]}
 GetLevelInstanceCount {nodes [
   n1 = GetLevelInstanceCount!() -> (then, Asset:String) {is=title K2Node_FunctionEntry}
   n2 = Find(TargetMap=Get AssetToInstanceIndexMap(), Key=@n1)
   > n3 = Branch(Condition=@n2.ReturnValue)
   n4 = "Get (a ref)"(Array=Get AssetInstances(), Dimension 1=@n2)
   > Return Node!(
    execute
    Count:int=Length(TargetArray=Break Find Class Usages Level Instance(FindClassUsages_LevelInstance=@n4).Instances_8_#2)
   )
   Return Node(execute=@n3.else, Count=0)
  ]}
 GetRequiredFilters {nodes [
   GetRequiredFilters() {is=title K2Node_FunctionEntry}
   n1 = ParseIntoArray(SourceString=Get RequiredFilter() -> (RequiredFilter:String), Delimiter=",")
   > n2 = For Each Loop(Array:Array of Strings=@n1) -> (Array Element:String)
   Return Node!(execute=@n2.Completed, RequiredFilters:Array of Strings=@n1)
   > PrintString(InString=Append!(A="Required Filter: ", B=@n2) -> (ReturnValue))
  ]}
 GetTimeEstimateRangeText {nodes [
   n1 = GetTimeEstimateRangeText!() -> (then, NumItems:int, MinTimePerItem:double=~, MaxTimeperItem:double=~) {is=title K2Node_FunctionEntry}
   > n2 = Branch(Condition="integer > integer"(A=@n1))
   n3 = "To Float (Integer)"(InInt=@n1)
   n4 = float / float(A=float * float(A=@n3, B=@n1.MaxTimeperItem) {is=title K2Node_PromotableOperator}, B=60)
   n5 = "To String (Text)"(InText=GetTimespanAsText!(
     Timespan:"Timespan Structure (by ref)"=From Minutes!(Minutes:double=@n4) -> (ReturnValue:Timespan Structure)
     bShowHoursWhenZero:bool=true
     bShowSeconds:bool=true
     FormatType:ETimespanAsTextFormat Enum=DaysHoursMinutesSeconds
     UnitDisplayType:ETimespanUnitDisplayFormat Enum=Abbreviated
     bUseSentenceCase:bool=false
    ) -> (ReturnValue:Text))
   n6 = IsEmpty(InString=@n5)
   n7 = float / float(A=float * float(A=@n3, B=@n1.MinTimePerItem), B=60)
   n8 = "To String (Text)"(InText=GetTimespanAsText(
     Timespan=From Minutes(Minutes=@n7)
     bShowHoursWhenZero=true
     bShowSeconds=true
     FormatType=DaysHoursMinutesSeconds
     UnitDisplayType=Abbreviated
     bUseSentenceCase=false
    ))
   n9 = IsEmpty(InString=@n8)
   > Return Node!(
    execute
    EstimateText:String=Append!(A=" (", B=SelectString(A=0s, B=@n8, bPickA=@n9), C:String=" to ", D:String=SelectString(A=0s, B=@n5, bPickA=@n6), E:String=")") -> (ReturnValue)
   )
   Return Node(execute=@n2.else)
  ]}
 GetTimeEstimateText {nodes [
   n1 = GetTimeEstimateText!() -> (then, NumItems:int, TimePerItem:double=~) {is=title K2Node_FunctionEntry}
   > n2 = Branch(Condition="integer > integer"(A=@n1))
   n3 = float / float(A=float * float(A="To Float (Integer)"(InInt=@n1), B=@n1.TimePerItem), B=60)
   n4 = "To String (Text)"(InText=GetTimespanAsText(
     Timespan=From Minutes(Minutes=@n3)
     bShowHoursWhenZero=true
     bShowSeconds=true
     FormatType=DaysHoursMinutesSeconds
     UnitDisplayType=Abbreviated
     bUseSentenceCase=false
    ))
   > Return Node(EstimateText=SelectString(A=0s, B=@n4, bPickA=IsEmpty(InString=@n4)))
   Return Node(execute=@n2.else)
  ]}
 HandleAssetViewChanged {nodes [
   n1 = HandleAssetViewChanged!() -> (then, bClearActiveTargetClass:bool=~) {is=title K2Node_FunctionEntry}
   > n2 = Branch(Condition=@n1)
   > n3 = Set ActiveTargetClass(ActiveTargetClass:Soft Class Path Structure) -> (Output_Get:Soft Class Path Structure)
   n4 = Branch(execute=[@n3.then, @n2.else], Condition=Get bUseTargetClassPath() -> (bUseTargetClassPath:bool))
   > n5 = Update Assets!(execute, self, SoftClassPath:Soft Class Path Structure=MakeSoftClassPath(PathString=Get TargetClassPath() -> (TargetClassPath:String))) -> (then)
   RunFilter(execute=[@n5.then, @n6.then])
   n6 = Update Assets(execute=@n4.else, SoftClassPath=Get TargetClass() -> (TargetClass:Soft Class Path Structure))
  ]}
 IsDataAssetClass {nodes [
   n1 = IsDataAssetClass!() -> (then, Class:Object Class Reference) {is=title K2Node_FunctionEntry}
   > Return Node!(execute, ReturnValue:bool=IsDataAsset!(Class:Object Class Reference=@n1) -> (ReturnValue:bool))
  ]}
 IsDataAssetClassPath {nodes [
   n1 = IsDataAssetClassPath!() -> (then, ClassPath:String) {is=title K2Node_FunctionEntry}
   > n2 = LoadClassAsset_Blocking(AssetClass=To Soft Class Reference(SoftClassPath=MakeSoftClassPath(PathString=@n1)))
   > Return Node!(execute, bResult:bool=IsDataAssetClass(Class=@n2) {is=title K2Node_CallFunction})
  ]}
 IsDataTableClass {nodes [
   n1 = IsDataTableClass!() -> (then, Class:Object Class Reference) {is=title K2Node_FunctionEntry}
   > n2 = Cast To DataTable Class!(execute:Exec, Class:Object Class Reference=@n1) -> (then:Exec, CastFailed:Exec, AsData Table:Data Table Class Reference) {is=title K2Node_ClassDynamicCast}
   > Return Node!(execute, ReturnValue:bool=true)
   Return Node(execute=@n2.CastFailed, ReturnValue=false)
  ]}
 IsDataTableClassPath {nodes [
   n1 = IsDataTableClassPath!() -> (then, ClassPath:String) {is=title K2Node_FunctionEntry}
   > n2 = LoadClassAsset_Blocking(AssetClass=To Soft Class Reference(SoftClassPath=MakeSoftClassPath(PathString=@n1)))
   > Return Node!(execute, bResult:bool=IsDataTableClass(Class=@n2) {is=title K2Node_CallFunction})
  ]}
 IsProcessingFinished {nodes [
   IsProcessingFinished() {is=title K2Node_FunctionEntry}
   n1 = Get ProcessingAssets()
   > Return Node!(execute, ReturnValue:bool="integer >= integer"(A=Get ProcessingIndex() -> (ProcessingIndex=0), B=Length(TargetArray:Array of Strings=@n1)))
  ]}
 Prepare Property Report {nodes [
   Prepare Property Report() {is=title K2Node_FunctionEntry}
   n1 = Get LogChangedProperties() -> (LogChangedProperties:String)
   > n2 = CreateClassPropertyReport!(
    execute
    BaseClassPath:String=BreakSoftClassPath!(InSoftClassPath:Soft Class Path Structure=Get ActiveTargetClass() -> (ActiveTargetClass:Soft Class Path Structure)) -> (PathString:String)
    Config:Class Property Report Config Structure=Make Class Property Report Config!(
     bCheckTimesDefaultChanged:bool=Get bCheckTimesDefaultChanged()
     bOnlyIncludeEditable:bool=Get bOnlyShowEditable()
     bIncludeInstanceEditable:bool=Get bShowIntanceEditable() -> (bShowIntanceEditable:bool)
     bIncludeBPModifiable:bool=Get bShowBPModifiable() -> (bShowBPModifiable:bool)
     bIncludeExposedInSpawn:bool=Get bShowExposedInSpawn() -> (bShowExposedInSpawn:bool)
     VerboseLoggedChangedProperties:Array of Strings=ParseIntoArray(SourceString=@n1, Delimiter=",")
    ) -> (ClassPropertyReportConfig:Class Property Report Config Structure) {is=title K2Node_MakeStruct}
   ) -> (then, bValid:bool, ReturnValue:Class Property Report Data Structure)
   > Set PropertyReport(PropertyReport:Class Property Report Data Structure=@n2.ReturnValue) -> (Output_Get:Class Property Report Data Structure)
   > PrepareProcessing()
   > Return Node!(execute, bValid:bool=@n2)
  ]}
 PrepareCustomFilter {nodes [
   PrepareCustomFilter() {is=title K2Node_FunctionEntry}
   > RunFilter()
   > PrepareProcessing()
   > Clear(TargetArray=Get VisibleAssets())
  ]}
 PrepareProcessing {nodes [
   PrepareProcessing() {is=title K2Node_FunctionEntry}
   n1 = Get VisibleAssets()
   > Branch(Condition=Is Not Empty!(TargetArray=@n1) -> (ReturnValue:bool))
   > Set ProcessingAssets(ProcessingAssets:Array of Strings=@n1) -> (Output_Get:Array of Strings)
   > Set ProcessingIndex(ProcessingIndex=0) -> (Output_Get=0)
   > Set TotalProcessingTimeSeconds(TotalProcessingTimeSeconds:double=0) -> (Output_Get:double)
   > Set LastDelayTime(LastDelayTime:Date Time Structure=Now!() -> (ReturnValue:Date Time Structure)) -> (Output_Get:Date Time Structure)
  ]}
 PrepareSizeReport {nodes [
   PrepareSizeReport() {is=title K2Node_FunctionEntry}
   > n1 = CreateClassSizeReport!(
    execute
    BaseClassPath:String=BreakSoftClassPath(InSoftClassPath=Get ActiveTargetClass())
    Config:Class Size Report Config Structure=Make Class Size Report Config!(EditorSizeOffset:int=Get EditorClassSizeOffset() -> (EditorClassSizeOffset:int=0)) -> (ClassSizeReportConfig:Class Size Report Config Structure) {is=title K2Node_MakeStruct}
   ) -> (then, bValid:bool, ReturnValue:Class Size Report Data Structure)
   > Set SizeReport(SizeReport:Class Size Report Data Structure=@n1.ReturnValue) -> (Output_Get:Class Size Report Data Structure)
   > PrepareProcessing() {is=title K2Node_CallFunction}
   > Return Node(bValid=@n1)
  ]}
 RefreshVisibleAssets {nodes [
   RefreshVisibleAssets() {is=title K2Node_FunctionEntry}
   n1 = Get VisibleAssets()
   > "SetText (Text Box)"!(
    execute
    self:Text Box Object Reference=Get txtCount() -> (txtCount:Text Box Object Reference)
    InText:Text="To Text (Integer)"(Value=Length(TargetArray=@n1))
   ) -> (then)
   > UpdatePageItems()
  ]}
 RunFilter {nodes [
   RunFilter() {is=title K2Node_FunctionEntry}
   > Clear(TargetArray=Get VisibleAssets())
   > n1 = GetRequiredFilters!(execute, self) -> (then, RequiredFilters:Array of Strings) {is=title K2Node_CallFunction}
   > n2 = GetExcludedFilters!(execute, self) -> (then, ExcludedFilters:Array of Strings) {is=title K2Node_CallFunction}
   > n3 = For Each Loop(Array=Get AllAssets())
   RefreshVisibleAssets(execute=@n3.Completed) {is=title K2Node_CallFunction}
   > UpdateCustomFilterApplied(bIsApplied=false)
   n4 = ApplyFilter!(execute=@n3.LoopBody, self, AssetPath:String=@n3, RequiredFIlters:Array of Strings=@n1, ExcludedFilters:Array of Strings=@n2) -> (then, bVisible:bool) {is=title K2Node_CallFunction}
   > Branch(Condition=@n4)
   > Add(TargetArray=Get VisibleAssets(), NewItem=@n3)
  ]}
 SelectLevelInstances {nodes [
   n1 = SelectLevelInstances!() -> (then, Asset:String) {is=title K2Node_FunctionEntry}
   n2 = Find(TargetMap=Get AssetToInstanceIndexMap(), Key=@n1)
   > Branch(Condition=@n2.ReturnValue)
   n3 = Get AssetInstances()
   > SetSelectedLevelActors(
    self=Get EditorActorSubsystem()
    ActorsToSelect=Break Find Class Usages Level Instance(FindClassUsages_LevelInstance="Get (a ref)"(Array=@n3, Dimension 1=@n2)).Instances_8_#2
   )
  ]}
 SetAssetSource {nodes [
   n1 = SetAssetSource!() -> (then, bLoadedLevel:bool=~) {is=title K2Node_FunctionEntry}
   > Set bUseLoadedLevel(bUseLoadedLevel:bool=@n1) -> (Output_Get:bool)
   > UpdateAssetSourceWidgets()
   > UpdateBottomPanelVisibility()
   > HandleAssetViewChanged(bClearActiveTargetClass=true) {is=title K2Node_CallFunction}
  ]}
 SetProcessingWidgetStates {nodes [
   n1 = SetProcessingWidgetStates!() -> (then, bProcessing:bool=~) {is=title K2Node_FunctionEntry}
   > n2 = Branch(Condition=@n1)
   > SetVisibility(
    self=[
     Get btnAdvanced() -> (btnAdvanced:TA Generic 01 Button Object Reference)
     Get LstAssetPaths()
     Get PropRequiredFilter()
     Get PropExcludedFilter()
     Get btnPageLeft() -> (btnPageLeft:TA Generic 01 Button Object Reference)
     Get btnPageRight() -> (btnPageRight:TA Generic 01 Button Object Reference)
     Get HbFilterSetup() -> (HbFilterSetup:Horizontal Box Object Reference)
     Get HbTargetSetup() -> (HbTargetSetup:Horizontal Box Object Reference)
    ]
    InVisibility=HitTestInvisible
   )
   > SetVisibility(self=Get btnCancelExport() -> (btnCancelExport:TA Generic 01 Button Object Reference), InVisibility=Visible)
   > SetVisibility(
    self=[
     Get VbGeneratePropertyReport() -> (VbGeneratePropertyReport:Vertical Box Object Reference)
     Get VbGenerateSizeReport() -> (VbGenerateSizeReport:Vertical Box Object Reference)
    ]
    InVisibility=Collapsed
   )
   > SetVisibility(self=Get txtProgress() -> (txtProgress:Text Box Object Reference), InVisibility=HitTestInvisible)
   > SetVisibility(self=Get PropTargetClass(), InVisibility=Select(Option 0=HitTestInvisible, Option 1=Collapsed, Index=Get bUseTargetClassPath()))
   > SetVisibility(self=Get PropTargetClassPath(), InVisibility=Select(Option 0=Collapsed, Option 1=HitTestInvisible, Index=Get bUseTargetClassPath()))
   SetVisibility(
    execute=@n2.else
    self=[
     Get btnAdvanced()
     Get btnPageLeft()
     Get LstAssetPaths()
     Get btnPageRight()
     Get PropRequiredFilter()
     Get HbFilterSetup()
     Get HbTargetSetup()
     Get PropExcludedFilter()
    ]
    InVisibility=Visible
   )
   > SetVisibility(self=Get btnCancelExport(), InVisibility=Collapsed)
   > SetVisibility(self=[Get VbGeneratePropertyReport(), Get VbGenerateSizeReport()], InVisibility=Visible)
   > SetVisibility(self=Get txtProgress(), InVisibility=Collapsed)
   > SetVisibility(self=Get PropTargetClass(), InVisibility=Select(Option 0=Visible, Option 1=Collapsed, Index=Get bUseTargetClassPath()))
   > SetVisibility(self=Get PropTargetClassPath(), InVisibility=Select(Option 0=Collapsed, Option 1=Visible, Index=Get bUseTargetClassPath()))
  ]}
 SetPropertyWidgetBinding {nodes [SetPropertyWidgetBinding!() -> (then, Widget:Single Property View Object Reference, Property:Name) {is=title K2Node_FunctionEntry}]}
 ShouldDelayProcessing {nodes [
   Inputs!() -> (In:Exec)
   n1 = Get LastDelayTime() -> (LastDelayTime:Date Time Structure)
   n2 = Now()
   n3 = Get Total Seconds!(A:Timespan Structure=Timespan - Timespan(
     A:Timespan Structure=Get Time Of Day!(A:Date Time Structure=@n2) -> (ReturnValue:Timespan Structure) {is=title K2Node_CallFunction}
     B:Timespan Structure=Get Time Of Day(A=@n1)
    ) -> (ReturnValue:Timespan Structure) {is=title K2Node_PromotableOperator}) -> (ReturnValue:double) {is=title K2Node_CallFunction}
   > n4 = Branch(Condition="float > float"(A=@n3, B=Get MaxTimeBetweenDelays() -> (MaxTimeBetweenDelays:double)))
   > Set TotalProcessingTimeSeconds(TotalProcessingTimeSeconds=float + float(A=@n3, B=Get TotalProcessingTimeSeconds() -> (TotalProcessingTimeSeconds:double)))
   > Set LastDelayTime(LastDelayTime=Now())
   > Outputs!(DoDelay:Exec, DoNotDelay:Exec=@n4.else)
  ]}
 ToggleAdvancedSelection {nodes [
   ToggleAdvancedSelection() {is=title K2Node_FunctionEntry}
   > Set bUseTargetClassPath(bUseTargetClassPath:bool=NOT Boolean(A=Get bUseTargetClassPath())) -> (Output_Get:bool)
   > SetVisibility(self=Get PropTargetClass(), InVisibility=Select(Option 0=Visible, Option 1=Collapsed, Index=Get bUseTargetClassPath()))
   > SetVisibility(self=Get PropTargetClassPath(), InVisibility=Select(Option 0=Collapsed, Option 1=Visible, Index=Get bUseTargetClassPath()))
   > setLabel!(
    execute
    self:TA Base Editor Button Widget Object Reference=Get btnAdvanced()
    label:String=SelectString(A=Basic, B=Advanced, bPickA=Get bUseTargetClassPath())
   ) -> (then)
   > n1 = Branch(Condition=Get bUseTargetClassPath())
   > Set TargetClassPath(TargetClassPath:String=BreakSoftClassPath(InSoftClassPath=Get TargetClass())) -> (Output_Get:String)
   Set TargetClass(execute=@n1.else, TargetClass:Soft Class Path Structure=MakeSoftClassPath(PathString=Get TargetClassPath())) -> (Output_Get:Soft Class Path Structure)
  ]}
 Update Assets {nodes [
   n1 = Update Assets!() -> (then, SoftClassPath:Soft Class Path Structure) {is=title K2Node_FunctionEntry}
   n2 = Get ActiveTargetClass()
   > Branch(Condition="Not Equal, Case Insenstive (String)"!(A:String=BreakSoftClassPath(InSoftClassPath=@n1), B:String=BreakSoftClassPath(InSoftClassPath=@n2)) -> (ReturnValue:bool))
   > n3 = Set ActiveTargetClass(ActiveTargetClass=@n1)
   n4 = BreakSoftClassPath(InSoftClassPath=@n3)
   > n5 = CalculateEditorOnlyPropertiesSize!(execute, ClassPath:String=@n4) -> (then, ReturnValue:int=0)
   > n6 = Negate Int!(" ":Exec, Value:"Integer (by ref)"=@n5) -> ("  ":Exec, Result:int) {is=title K2Node_MacroInstance}
   > Set EditorClassSizeOffset(EditorClassSizeOffset:int=@n6) -> (Output_Get:int=0)
   > Clear(TargetArray=Get AllAssets())
   > Clear!(execute, TargetMap:Map of Strings to Integers=Get AssetToInstanceIndexMap()) -> (then) {is=title K2Node_CallFunction}
   > Clear(TargetArray:Array of Find Class Usages Level Instance Structures=Get AssetInstances()) {is=title K2Node_CallArrayFunction}
   > LogString(InString=@n4)
   > n7 = Branch(Condition=Get bUseLoadedLevel())
   > n8 = UpdateLevelAssets!(execute, self, SoftClassPath:Soft Class Path Structure=@n3) -> (then)
   FinshedUpdatingAssets(execute=[@n9.then, @n8.then]) {is=title K2Node_CallFunction}
   n9 = UpdateClassAssets!(execute=@n7.else, self, ClassPath:Soft Class Path Structure=@n3) -> (then)
  ]}
 UpdateAssetSourceWidgets {nodes [
   UpdateAssetSourceWidgets() {is=title K2Node_FunctionEntry}
   > n1 = Branch(Condition=Get bUseLoadedLevel())
   n2 = Get btnLoadedLevel() -> (btnLoadedLevel:TA Generic 01 Button Object Reference)
   > Set buttonColor_Default!(execute, buttonColor_Default:Linear Color Structure=(R=0.0635,G=1,B=0,A=1), self:TA Base Editor Button Widget Object Reference=@n2) -> (then, Output_Get:Linear Color Structure)
   > Set buttonColor_Selected!(execute, buttonColor_Selected:Linear Color Structure=(R=0.063,G=1,B=0,A=1), self:TA Base Editor Button Widget Object Reference=@n2) -> (then, Output_Get:Linear Color Structure)
   > setSelected!(execute, self:TA Base Editor Button Widget Object Reference=@n2, selected:bool=false, throwEvent:bool=true) -> (then)
   > "SetForegroundColor (Multi-Line Text Box)"!(
    execute
    self:"Text Box (Multi- Line) Object Reference"=Get txtLoadedLevel() -> (txtLoadedLevel:Editor Utility Multi Line Editable Text Box Object Reference)
    color:Linear Color Structure=(R=1,G=1,B=1,A=1)
   ) -> (then)
   n3 = Get btnProjectAssets() -> (btnProjectAssets:TA Generic 01 Button Object Reference)
   > Set buttonColor_Default(buttonColor_Default=(R=1,G=0,B=0.0367,A=1), self=@n3)
   > Set buttonColor_Selected(buttonColor_Selected=(R=1,G=0,B=0.0369,A=1), self=@n3)
   > setSelected(self=@n3, selected=true, throwEvent=true)
   > "SetForegroundColor (Multi-Line Text Box)"(
    self=Get txtProjectAssets() -> (txtProjectAssets:Editor Utility Multi Line Editable Text Box Object Reference)
    color=(R=0.2135,G=0.2135,B=0.2135,A=1)
   )
   > "SetText (Multi-Line Text Box)"!(
    execute
    self:"Text Box (Multi- Line) Object Reference"=Get txtWarning() -> (txtWarning:Editor Utility Multi Line Editable Text Box Object Reference)
    InText:Text="LOCTEXT(\"Select the parent class to find all of the usages of in the currently loaded level.\")"
   ) -> (then)
   n4 = Get btnLoadedLevel()
   Set buttonColor_Default(execute=@n1.else, buttonColor_Default=(R=1,G=0,B=0.0369,A=1), self=@n4)
   > Set buttonColor_Selected(buttonColor_Selected=(R=1,G=0,B=0.0369,A=1), self=@n4)
   > setSelected(self=@n4, selected=true, throwEvent=true)
   > "SetForegroundColor (Multi-Line Text Box)"(self=Get txtLoadedLevel(), color=(R=0.2122,G=0.2122,B=0.2122,A=1))
   n5 = Get btnProjectAssets()
   > Set buttonColor_Default(buttonColor_Default=(R=0.063,G=1,B=0,A=1), self=@n5)
   > Set buttonColor_Selected(buttonColor_Selected=(R=0.063,G=1,B=0,A=1), self=@n5)
   > setSelected(self=@n5, selected=false, throwEvent=true)
   > "SetForegroundColor (Multi-Line Text Box)"(self=Get txtProjectAssets(), color=(R=1,G=1,B=1,A=1))
   > "SetText (Multi-Line Text Box)"(
    self=Get txtWarning()
    InText="LOCTEXT(\"Select a parent class to find all blueprint classes derive from it. (NOTE: Selcting a very common parent class will take a long time to update the list)\")"
   )
  ]}
 UpdateBottomPanelVisibility {nodes [
   UpdateBottomPanelVisibility() {is=title K2Node_FunctionEntry}
   > SetVisibility(
    self=Get PnlAssetControls() -> (PnlAssetControls:Horizontal Box Object Reference)
    InVisibility=Select(Option 0=SelfHitTestInvisible, Option 1=Collapsed, Index=Get bUseLoadedLevel())
   )
  ]}
 UpdateClassAssets {nodes [
   n1 = UpdateClassAssets!() -> (then, ClassPath:Soft Class Path Structure) {is=title K2Node_FunctionEntry}
   n2 = BreakSoftClassPath(InSoftClassPath=@n1)
   n3 = Split(SourceString=@n2, InStr=., SearchDir=FromStart)
   > n4 = GetDerivedClasses!(
    execute
    self
    ClassName:"Top Level Asset Path Structure (by ref)"=MakeTopLevelAssetPath(PackageName=@n3, AssetName=@n3.RightS)
    ClassPath:String=@n2
   ) -> (then, Result:Array of Top Level Asset Path Structures) {is=title K2Node_CallFunction}
   > n5 = For Each Loop(Array:Array of Top Level Asset Path Structures=@n4) -> (Array Element:Top Level Asset Path Structure)
   > AddAssetPath(TopLevelAssetPath=@n5)
  ]}
 UpdateCustomFilterApplied {nodes [
   n1 = UpdateCustomFilterApplied!() -> (then, bIsApplied:bool=~) {is=title K2Node_FunctionEntry}
   > n2 = Set bCustomFilterApplied(bCustomFilterApplied:bool=@n1) -> (Output_Get:bool)
   > setLabel(
    self=Get btnApplyCustomFilter() -> (btnApplyCustomFilter:TA Generic 01 Button Object Reference)
    label=SelectString(A=Clear Custom Filter, B=Apply Custom Filter, bPickA=@n2)
   )
   > Set buttonColor_Default(buttonColor_Default=SelectColor(A=(R=1,G=0.0301,B=0,A=1), B=(R=0.02,G=0.4,B=0,A=1), bPickA=@n2), self=Get btnApplyCustomFilter())
  ]}
 UpdateLevelAssets {nodes [
   n1 = UpdateLevelAssets!() -> (then, SoftClassPath:Soft Class Path Structure) {is=title K2Node_FunctionEntry}
   > n2 = LoadClassAsset_Blocking(AssetClass=To Soft Class Reference(SoftClassPath=@n1))
   > n3 = GetAllLevelActors(self=Get EditorActorSubsystem())
   > AddLevelActors(Actors=@n3, TargetClass=@n2)
  ]}
 UpdatePageItems {nodes [
   UpdatePageItems() {is=title K2Node_FunctionEntry}
   > ValidatePageIndex()
   n1 = Get CurrentPageIndexOffset() -> (CurrentPageIndexOffset=0)
   n2 = Get MaxItemsPerPage() -> (MaxItemsPerPage=0)
   > n3 = For Loop(FirstIndex=@n1, LastIndex=int - int(A=int + int(A=@n1, B=@n2)))
   UpdatePageText(execute=@n3.Completed)
   > SetScrollOffset!(
    execute
    self:Scroll Box Object Reference=Get ScrollBox_List!(self:TA Generic 01 List Object Reference=Get LstAssetPaths()) -> (ScrollBox_List:Scroll Box Object Reference)
    NewScrollOffset:float=0
   ) -> (then)
   n4 = getItemAtIndex!(
    execute=@n3.LoopBody
    self:TA Generic 01 List Object Reference=Get LstAssetPaths()
    itemIndex:int="% (Integer)"(A=@n3, B=Get MaxItemsPerPage() -> (MaxItemsPerPage=0)) -> (ReturnValue=0)
   ) -> (then, item:User Widget Object Reference)
   > Is Valid(InputObject=@n4)
   > n5 = Cast To TA_Generic01_ListItem(Object=@n4)
   > n6 = Branch(Condition=Is Valid Index(TargetArray=Get VisibleAssets(), IndexToTest=@n3))
   > SetVisibility(self=@n5, InVisibility=Visible)
   n7 = Get VisibleAssets()
   > setLabel(
    self=@n5
    label=GetAssetDisplayString!(Asset:String="Get (a ref)"(Array:Array of Strings=@n7, Dimension 1=@n3) -> (Output:"String (by ref)")) -> (DisplayString:String) {is=title K2Node_CallFunction}
   )
   SetVisibility(execute=@n6.else, self=@n5, InVisibility=Collapsed)
  ]}
 UpdatePageText {nodes [
   UpdatePageText() {is=title K2Node_FunctionEntry}
   n1 = Get CurrentPageIndexOffset() -> (CurrentPageIndexOffset=0)
   n2 = "To String (Integer)"(InInt=int + int(A=@n1))
   n3 = Length(TargetArray=Get VisibleAssets())
   n4 = Get MaxItemsPerPage() -> (MaxItemsPerPage=0)
   n5 = "To String (Integer)"(InInt="Min (Integer)"(A=int + int(A=@n1, B=@n4), B=@n3) -> (ReturnValue=0))
   n6 = "To String (Integer)"(InInt=@n3)
   > "SetText (Text Box)"(
    self=Get txtPageCount() -> (txtPageCount:Text Box Object Reference)
    InText="To Text (String)"(InString=Append(A=@n2, B=-, C=@n5, D=" of ", E=@n6))
   )
  ]}
 UpdateProcessingStatusText {nodes [
   n1 = UpdateProcessingStatusText!() -> (then, Index:int) {is=title K2Node_FunctionEntry}
   n2 = Length(TargetArray=Get ProcessingAssets())
   n3 = SafeDivide(A="To Float (Integer)"(InInt=@n1), B="To Float (Integer)"(InInt=@n2))
   n4 = "To String (Integer64)"!(InInt:Integer64=Round to Integer64!(A:double=float * float(A=@n3, B=100)) -> (ReturnValue:Integer64=0)) -> (ReturnValue:String)
   n5 = "To String (Integer)"(InInt=@n1)
   n6 = "To String (Integer)"(InInt=@n2)
   n7 = SafeDivide(A=Get TotalProcessingTimeSeconds(), B="To Float (Integer)"(InInt=Get ProcessingIndex() -> (ProcessingIndex=0)))
   n8 = int - int(A=Length(TargetArray=Get ProcessingAssets()), B=Get ProcessingIndex() -> (ProcessingIndex=0))
   n9 = SelectString(
    A=Append!(
     A=" -- "
     B=GetTimeEstimateText!(NumItems:int=@n8, TimePerItem:double=@n7) -> (EstimateText:String) {is=title K2Node_CallFunction}
     C=" Remaining"
    ) -> (ReturnValue)
    bPickA="integer > integer"(A=Get ProcessingIndex() -> (ProcessingIndex=0))
   )
   > "SetText (Text Box)"(
    self=Get txtProgress()
    InText="To Text (String)"(InString=Append!(A=@n5, B=/, C=@n6, D:String=" -- ", E:String=@n4, F:String=%, G:String=@n9) -> (ReturnValue))
   )
  ]}
 ValidatePageIndex {nodes [
   ValidatePageIndex() {is=title K2Node_FunctionEntry}
   n1 = Get VisibleAssets()
   > While Loop!(execute:Exec, Condition:bool="integer > integer"(A=Get CurrentPageIndexOffset() -> (CurrentPageIndexOffset=0), B=Length(TargetArray=@n1))) -> (LoopBody:Exec, Completed:Exec) {is=title K2Node_MacroInstance}
   > Set CurrentPageIndexOffset(CurrentPageIndexOffset=int - int(A=Get CurrentPageIndexOffset() -> (CurrentPageIndexOffset=0), B=Get MaxItemsPerPage() -> (MaxItemsPerPage=0))) -> (Output_Get=0)
  ]}
]
variables [
 ActiveTargetClass {type=SoftClassPath, default=""}
 AllAssets {type="array<string>", default []}
 AssetInstances {type="array<FindClassUsages_LevelInstance>", default []}
 AssetToInstanceIndexMap {type="map<integer>", default {}}
 bCheckTimesDefaultChanged {type=boolean, default=true}
 bCustomFilterApplied {type=boolean}
 bIncludeNativeClasses {type=boolean, default=true}
 bOnlyShowEditable {type=boolean, default=true}
 bOutputPropertyReportAsCSV {type=boolean}
 bOutputSIzeReportAsCSV {type=boolean}
 bShowBPModifiable {type=boolean}
 bShowExposedInSpawn {type=boolean}
 bShowIntanceEditable {type=boolean}
 bUseLoadedLevel {type=boolean}
 bUseTargetClassPath {type=boolean}
 ClassNamesSet {type="array<TopLevelAssetPath>", default []}
 CurrentPageIndexOffset {type=integer}
 EditorClassSizeOffset {type=integer}
 ExcludedFilter {type=string, default=""}
 ExportOutputFolder {type=DirectoryPath, default {path=""}}
 LastDelayTime {type=DateTime, default=0001.01.01-00.00.00}
 LogChangedProperties {type=string, default=""}
 MaxItemsPerPage {type=integer, default=100}
 MaxSizeReportEstimatePerAsset {type=number, default=0.65}
 MaxTimeBetweenDelays {type=number, default=1}
 MinSizeReportEstimatePerAsset {type=number, default=0.001}
 ProcessingAssets {type="array<string>", default []}
 ProcessingIndex {type=integer}
 PropertyReport {type=ClassPropertyReportData, default {}}
 RequiredFilter {type=string, default=""}
 SizeReport {type=ClassSizeReportData, default {}}
 TargetClass {type=SoftClassPath, default=""}
 TargetClassPath {type=string, default=""}
 TotalProcessingTimeSeconds {type=number}
 VisibleAssets {type="array<string>", default []}
]
widgets {
 info {rootWidgetClass=&VerticalBox, widgetCount=50}
 widgets [
  VerticalBox_0 {widgetClassPath=&VerticalBox, widget=VerticalBox_0}
  txtWarning {
   widgetClassPath=/Script/Blutility.EditorUtilityMultiLineEditableTextBox
   bIsVariable=true
   parent=VerticalBox_0
   widget=txtWarning
   properties {
    bIsReadOnly=true
    justification=Center
    text="Select a parent class to find all blueprint classes derive from it. (NOTE: Selcting a very common parent class will take a long time to update the list)"
    visibility=Collapsed
    widgetStyle {
     backgroundColor {colorUseRule=UseColor_Specified, specifiedColor=<rgba 1 1 1 1>}
     backgroundImageFocused {
      bIsDynamicallyLoaded=false
      drawAs=RoundedBox
      imageType=NoImage
      mirroring=NoMirror
      resourceName=None
      resourceObject=null
      tiling=NoTile
      imageSize=<0 0>
      margin=<ltrb 0 0 0 0>
      outlineSettings {
       bUseBrushTransparency=false
       roundingType=FixedRadius
       width=1
       color {colorUseRule=UseColor_ColorTable, specifiedColor=<rgba 0 0 0 0>}
       cornerRadii=<4 4 4 4>
      }
      tintColor {colorUseRule=UseColor_Specified, specifiedColor=<rgba 0.0048 0.0048 0.0048 1>}
      uVRegion {bIsValid=false, max=<0 0>, min=<0 0>}
     }
     backgroundImageHovered {
      bIsDynamicallyLoaded=false
      drawAs=RoundedBox
      imageType=NoImage
      mirroring=NoMirror
      resourceName=None
      resourceObject=null
      tiling=NoTile
      imageSize=<0 0>
      margin=<ltrb 0 0 0 0>
      outlineSettings {
       bUseBrushTransparency=false
       roundingType=FixedRadius
       width=1
       color {colorUseRule=UseColor_ColorTable, specifiedColor=<rgba 0 0 0 0>}
       cornerRadii=<4 4 4 4>
      }
      tintColor {colorUseRule=UseColor_Specified, specifiedColor=<rgba 0.0048 0.0048 0.0048 1>}
      uVRegion {bIsValid=false, max=<0 0>, min=<0 0>}
     }
     backgroundImageNormal {
      bIsDynamicallyLoaded=false
      drawAs=RoundedBox
      imageType=NoImage
      mirroring=NoMirror
      resourceName=None
      resourceObject=null
      tiling=NoTile
      imageSize=<0 0>
      margin=<ltrb 0 0 0 0>
      outlineSettings {
       bUseBrushTransparency=false
       roundingType=FixedRadius
       width=1
       color {colorUseRule=UseColor_ColorTable, specifiedColor=<rgba 0 0 0 0>}
       cornerRadii=<4 4 4 4>
      }
      tintColor {colorUseRule=UseColor_Specified, specifiedColor=<rgba 0.0048 0.0048 0.0048 1>}
      uVRegion {bIsValid=false, max=<0 0>, min=<0 0>}
     }
     backgroundImageReadOnly {
      bIsDynamicallyLoaded=false
      drawAs=RoundedBox
      imageType=NoImage
      mirroring=NoMirror
      resourceName=None
      resourceObject=null
      tiling=NoTile
      imageSize=<0 0>
      margin=<ltrb 0 0 0 0>
      outlineSettings {
       bUseBrushTransparency=false
       roundingType=FixedRadius
       width=1
       color {colorUseRule=UseColor_ColorTable, specifiedColor=<rgba 0 0 0 0>}
       cornerRadii=<4 4 4 4>
      }
      tintColor {colorUseRule=UseColor_Specified, specifiedColor=<rgba 0 0 0 1>}
      uVRegion {bIsValid=false, max=<0 0>, min=<0 0>}
     }
     focusedForegroundColor {colorUseRule=UseColor_Specified, specifiedColor=<rgba 1 1 1 1>}
     foregroundColor {colorUseRule=UseColor_Specified, specifiedColor=<rgba 0.5271 0.5271 0.5271 1>}
     hScrollBarPadding=<ltrb 0 0 0 0>
     padding=<ltrb 8 3 8 4>
     readOnlyForegroundColor {colorUseRule=UseColor_Specified, specifiedColor=<rgba 1 1 1 1>}
     scrollBarStyle {
      thickness=8
      draggedThumbImage {
       bIsDynamicallyLoaded=false
       drawAs=RoundedBox
       imageType=NoImage
       mirroring=NoMirror
       resourceName=None
       resourceObject=null
       tiling=NoTile
       imageSize=<0 0>
       margin=<ltrb 0 0 0 0>
       outlineSettings {
        bUseBrushTransparency=false
        roundingType=FixedRadius
        width=0
        color {colorUseRule=UseColor_Specified, specifiedColor=<rgba 0 0 0 0>}
        cornerRadii=<4 4 4 4>
       }
       tintColor {colorUseRule=UseColor_Specified, specifiedColor=<rgba 0.2159 0.2159 0.2159 1>}
       uVRegion {bIsValid=false, max=<0 0>, min=<0 0>}
      }
      horizontalBackgroundImage {
       bIsDynamicallyLoaded=false
       drawAs=NoDrawType
       imageType=NoImage
       mirroring=NoMirror
       resourceName=None
       resourceObject=null
       tiling=NoTile
       imageSize=<0 0>
       margin=<ltrb 0 0 0 0>
       outlineSettings {
        bUseBrushTransparency=false
        roundingType=HalfHeightRadius
        width=0
        color {colorUseRule=UseColor_Specified, specifiedColor=<rgba 0 0 0 0>}
        cornerRadii=<0 0 0 0>
       }
       tintColor {colorUseRule=UseColor_Specified, specifiedColor=<rgba 1 1 1 1>}
       uVRegion {bIsValid=false, max=<0 0>, min=<0 0>}
      }
      horizontalBottomSlotImage {
       bIsDynamicallyLoaded=false
       drawAs=NoDrawType
       imageType=NoImage
       mirroring=NoMirror
       resourceName=None
       resourceObject=null
       tiling=NoTile
       imageSize=<0 0>
       margin=<ltrb 0 0 0 0>
       outlineSettings {
        bUseBrushTransparency=false
        roundingType=HalfHeightRadius
        width=0
        color {colorUseRule=UseColor_Specified, specifiedColor=<rgba 0 0 0 0>}
        cornerRadii=<0 0 0 0>
       }
       tintColor {colorUseRule=UseColor_Specified, specifiedColor=<rgba 1 1 1 1>}
       uVRegion {bIsValid=false, max=<0 0>, min=<0 0>}
      }
      horizontalTopSlotImage {
       bIsDynamicallyLoaded=false
       drawAs=NoDrawType
       imageType=NoImage
       mirroring=NoMirror
       resourceName=None
       resourceObject=null
       tiling=NoTile
       imageSize=<0 0>
       margin=<ltrb 0 0 0 0>
       outlineSettings {
        bUseBrushTransparency=false
        roundingType=HalfHeightRadius
        width=0
        color {colorUseRule=UseColor_Specified, specifiedColor=<rgba 0 0 0 0>}
        cornerRadii=<0 0 0 0>
       }
       tintColor {colorUseRule=UseColor_Specified, specifiedColor=<rgba 1 1 1 1>}
       uVRegion {bIsValid=false, max=<0 0>, min=<0 0>}
      }
      hoveredThumbImage {
       bIsDynamicallyLoaded=false
       drawAs=RoundedBox
       imageType=NoImage
       mirroring=NoMirror
       resourceName=None
       resourceObject=null
       tiling=NoTile
       imageSize=<0 0>
       margin=<ltrb 0 0 0 0>
       outlineSettings {
        bUseBrushTransparency=false
        roundingType=FixedRadius
        width=0
        color {colorUseRule=UseColor_Specified, specifiedColor=<rgba 0 0 0 0>}
        cornerRadii=<4 4 4 4>
       }
       tintColor {colorUseRule=UseColor_Specified, specifiedColor=<rgba 0.2159 0.2159 0.2159 1>}
       uVRegion {bIsValid=false, max=<0 0>, min=<0 0>}
      }
      normalThumbImage {
       bIsDynamicallyLoaded=false
       drawAs=RoundedBox
       imageType=NoImage
       mirroring=NoMirror
       resourceName=None
       resourceObject=null
       tiling=NoTile
       imageSize=<0 0>
       margin=<ltrb 0 0 0 0>
       outlineSettings {
        bUseBrushTransparency=false
        roundingType=FixedRadius
        width=0
        color {colorUseRule=UseColor_Specified, specifiedColor=<rgba 0 0 0 0>}
        cornerRadii=<4 4 4 4>
       }
       tintColor {colorUseRule=UseColor_Specified, specifiedColor=<rgba 0.0953 0.0953 0.0953 1>}
       uVRegion {bIsValid=false, max=<0 0>, min=<0 0>}
      }
      verticalBackgroundImage {
       bIsDynamicallyLoaded=false
       drawAs=NoDrawType
       imageType=NoImage
       mirroring=NoMirror
       resourceName=None
       resourceObject=null
       tiling=NoTile
       imageSize=<0 0>
       margin=<ltrb 0 0 0 0>
       outlineSettings {
        bUseBrushTransparency=false
        roundingType=HalfHeightRadius
        width=0
        color {colorUseRule=UseColor_Specified, specifiedColor=<rgba 0 0 0 0>}
        cornerRadii=<0 0 0 0>
       }
       tintColor {colorUseRule=UseColor_Specified, specifiedColor=<rgba 1 1 1 1>}
       uVRegion {bIsValid=false, max=<0 0>, min=<0 0>}
      }
      verticalBottomSlotImage {
       bIsDynamicallyLoaded=false
       drawAs=NoDrawType
       imageType=NoImage
       mirroring=NoMirror
       resourceName=None
       resourceObject=null
       tiling=NoTile
       imageSize=<0 0>
       margin=<ltrb 0 0 0 0>
       outlineSettings {
        bUseBrushTransparency=false
        roundingType=HalfHeightRadius
        width=0
        color {colorUseRule=UseColor_Specified, specifiedColor=<rgba 0 0 0 0>}
        cornerRadii=<0 0 0 0>
       }
       tintColor {colorUseRule=UseColor_Specified, specifiedColor=<rgba 1 1 1 1>}
       uVRegion {bIsValid=false, max=<0 0>, min=<0 0>}
      }
      verticalTopSlotImage {
       bIsDynamicallyLoaded=false
       drawAs=NoDrawType
       imageType=NoImage
       mirroring=NoMirror
       resourceName=None
       resourceObject=null
       tiling=NoTile
       imageSize=<0 0>
       margin=<ltrb 0 0 0 0>
       outlineSettings {
        bUseBrushTransparency=false
        roundingType=HalfHeightRadius
        width=0
        color {colorUseRule=UseColor_Specified, specifiedColor=<rgba 0 0 0 0>}
        cornerRadii=<0 0 0 0>
       }
       tintColor {colorUseRule=UseColor_Specified, specifiedColor=<rgba 1 1 1 1>}
       uVRegion {bIsValid=false, max=<0 0>, min=<0 0>}
      }
     }
     textStyle {
      overflowPolicy=Clip
      transformPolicy=None
      colorAndOpacity {colorUseRule=UseColor_Foreground, specifiedColor=<rgba 1 0 1 1>}
      font {
       bForceMonospaced=false
       bMaterialIsStencil=false
       fontMaterial=null
       fontObject=null
       letterSpacing=0
       monospacedWidth=1
       size=10
       skewAmount=0
       typefaceFontName=Regular
       outlineSettings {
        bApplyOutlineToDropShadows=false
        bMiteredCorners=false
        bSeparateFillAlpha=false
        outlineMaterial=null
        outlineSize=0
        outlineColor=<rgba 0 0 0 1>
       }
      }
      highlightColor {colorUseRule=UseColor_ColorTable, specifiedColor=<rgba 0 0 0 0>}
      highlightShape {
       bIsDynamicallyLoaded=false
       drawAs=Image
       imageType=NoImage
       mirroring=NoMirror
       resourceName=None
       resourceObject=null
       tiling=NoTile
       imageSize=<0 0>
       margin=<ltrb 0 0 0 0>
       outlineSettings {
        bUseBrushTransparency=false
        roundingType=HalfHeightRadius
        width=0
        color {colorUseRule=UseColor_Specified, specifiedColor=<rgba 0 0 0 0>}
        cornerRadii=<0 0 0 0>
       }
       tintColor {colorUseRule=UseColor_ColorTable, specifiedColor=<rgba 0 0 0 0>}
       uVRegion {bIsValid=false, max=<0 0>, min=<0 0>}
      }
      selectedBackgroundColor {colorUseRule=UseColor_ColorTable, specifiedColor=<rgba 0 0 0 0>}
      shadowColorAndOpacity=<rgba 0 0 0 1>
      shadowOffset=<0 0>
      strikeBrush {
       bIsDynamicallyLoaded=false
       drawAs=Image
       imageType=NoImage
       mirroring=NoMirror
       resourceName=None
       resourceObject=null
       tiling=NoTile
       imageSize=<32 32>
       margin=<ltrb 0 0 0 0>
       outlineSettings {
        bUseBrushTransparency=false
        roundingType=HalfHeightRadius
        width=0
        color {colorUseRule=UseColor_Specified, specifiedColor=<rgba 0 0 0 0>}
        cornerRadii=<0 0 0 0>
       }
       tintColor {colorUseRule=UseColor_Specified, specifiedColor=<rgba 1 1 1 1>}
       uVRegion {bIsValid=false, max=<0 0>, min=<0 0>}
      }
      underlineBrush {
       bIsDynamicallyLoaded=false
       drawAs=Image
       imageType=NoImage
       mirroring=NoMirror
       resourceName=None
       resourceObject=null
       tiling=NoTile
       imageSize=<32 32>
       margin=<ltrb 0 0 0 0>
       outlineSettings {
        bUseBrushTransparency=false
        roundingType=HalfHeightRadius
        width=0
        color {colorUseRule=UseColor_Specified, specifiedColor=<rgba 0 0 0 0>}
        cornerRadii=<0 0 0 0>
       }
       tintColor {colorUseRule=UseColor_Specified, specifiedColor=<rgba 1 1 1 1>}
       uVRegion {bIsValid=false, max=<0 0>, min=<0 0>}
      }
     }
     vScrollBarPadding=<ltrb 0 0 0 0>
    }
   }
   slot {class=VerticalBoxSlot, properties {size {sizeRule=Automatic, value=2}}}
  }
  HbDataSource {widgetClassPath=&HorizontalBox, bIsVariable=true, parent=VerticalBox_0, widget=HbDataSource, slot {class=VerticalBoxSlot}}
  VerticalBox_96 {
   widgetClassPath=&VerticalBox
   parent=HbDataSource
   widget=VerticalBox_96
   slot {class=HorizontalBoxSlot, properties {size {sizeRule=Fill, value=1}}}
  }
  btnProjectAssets {
   widgetClassPath=&TA_Generic01_Button
   parent=VerticalBox_96
   widget=btnProjectAssets
   properties {label=Project Assets, visibility=Visible, buttonColor_Default=<rgba 0.1166 1 0 1>}
   slot {class=VerticalBoxSlot, properties {padding=<ltrb 5 0 5 0>}}
  }
  txtProjectAssets {
   widgetClassPath=&EditorUtilityMultiLineEditableTextBox
   bIsVariable=true
   parent=VerticalBox_96
   widget=txtProjectAssets
   properties {
    bIsReadOnly=true
    justification=Center
    text="Class usages will be found from all the assets currently loaded in the project.  This will not find assets in unloaded Plugins.\r\nNOTE: Selcting a very common parent class will take a long time to update the list\r\nLeft-Clicking entries will select the asset."
    widgetStyle {
     backgroundColor {colorUseRule=UseColor_Specified, specifiedColor=<rgba 1 1 1 1>}
     backgroundImageFocused {
      bIsDynamicallyLoaded=false
      drawAs=RoundedBox
      imageType=NoImage
      mirroring=NoMirror
      resourceName=None
      resourceObject=null
      tiling=NoTile
      imageSize=<0 0>
      margin=<ltrb 0 0 0 0>
      outlineSettings {
       bUseBrushTransparency=false
       roundingType=FixedRadius
       width=1
       color {colorUseRule=UseColor_ColorTable, specifiedColor=<rgba 0 0 0 0>}
       cornerRadii=<4 4 4 4>
      }
      tintColor {colorUseRule=UseColor_Specified, specifiedColor=<rgba 0.0048 0.0048 0.0048 1>}
      uVRegion {bIsValid=false, max=<0 0>, min=<0 0>}
     }
     backgroundImageHovered {
      bIsDynamicallyLoaded=false
      drawAs=RoundedBox
      imageType=NoImage
      mirroring=NoMirror
      resourceName=None
      resourceObject=null
      tiling=NoTile
      imageSize=<0 0>
      margin=<ltrb 0 0 0 0>
      outlineSettings {
       bUseBrushTransparency=false
       roundingType=FixedRadius
       width=1
       color {colorUseRule=UseColor_ColorTable, specifiedColor=<rgba 0 0 0 0>}
       cornerRadii=<4 4 4 4>
      }
      tintColor {colorUseRule=UseColor_Specified, specifiedColor=<rgba 0.0048 0.0048 0.0048 1>}
      uVRegion {bIsValid=false, max=<0 0>, min=<0 0>}
     }
     backgroundImageNormal {
      bIsDynamicallyLoaded=false
      drawAs=RoundedBox
      imageType=NoImage
      mirroring=NoMirror
      resourceName=None
      resourceObject=null
      tiling=NoTile
      imageSize=<0 0>
      margin=<ltrb 0 0 0 0>
      outlineSettings {
       bUseBrushTransparency=false
       roundingType=FixedRadius
       width=1
       color {colorUseRule=UseColor_ColorTable, specifiedColor=<rgba 0 0 0 0>}
       cornerRadii=<4 4 4 4>
      }
      tintColor {colorUseRule=UseColor_Specified, specifiedColor=<rgba 0.0048 0.0048 0.0048 1>}
      uVRegion {bIsValid=false, max=<0 0>, min=<0 0>}
     }
     backgroundImageReadOnly {
      bIsDynamicallyLoaded=false
      drawAs=RoundedBox
      imageType=NoImage
      mirroring=NoMirror
      resourceName=None
      resourceObject=null
      tiling=NoTile
      imageSize=<0 0>
      margin=<ltrb 0 0 0 0>
      outlineSettings {
       bUseBrushTransparency=false
       roundingType=FixedRadius
       width=1
       color {colorUseRule=UseColor_ColorTable, specifiedColor=<rgba 0 0 0 0>}
       cornerRadii=<4 4 4 4>
      }
      tintColor {colorUseRule=UseColor_Specified, specifiedColor=<rgba 0 0 0 1>}
      uVRegion {bIsValid=false, max=<0 0>, min=<0 0>}
     }
     focusedForegroundColor {colorUseRule=UseColor_Specified, specifiedColor=<rgba 1 1 1 1>}
     foregroundColor {colorUseRule=UseColor_Specified, specifiedColor=<rgba 0.5271 0.5271 0.5271 1>}
     hScrollBarPadding=<ltrb 0 0 0 0>
     padding=<ltrb 8 3 8 4>
     readOnlyForegroundColor {colorUseRule=UseColor_Specified, specifiedColor=<rgba 1 1 1 1>}
     scrollBarStyle {
      thickness=8
      draggedThumbImage {
       bIsDynamicallyLoaded=false
       drawAs=RoundedBox
       imageType=NoImage
       mirroring=NoMirror
       resourceName=None
       resourceObject=null
       tiling=NoTile
       imageSize=<0 0>
       margin=<ltrb 0 0 0 0>
       outlineSettings {
        bUseBrushTransparency=false
        roundingType=FixedRadius
        width=0
        color {colorUseRule=UseColor_Specified, specifiedColor=<rgba 0 0 0 0>}
        cornerRadii=<4 4 4 4>
       }
       tintColor {colorUseRule=UseColor_Specified, specifiedColor=<rgba 0.2159 0.2159 0.2159 1>}
       uVRegion {bIsValid=false, max=<0 0>, min=<0 0>}
      }
      horizontalBackgroundImage {
       bIsDynamicallyLoaded=false
       drawAs=NoDrawType
       imageType=NoImage
       mirroring=NoMirror
       resourceName=None
       resourceObject=null
       tiling=NoTile
       imageSize=<0 0>
       margin=<ltrb 0 0 0 0>
       outlineSettings {
        bUseBrushTransparency=false
        roundingType=HalfHeightRadius
        width=0
        color {colorUseRule=UseColor_Specified, specifiedColor=<rgba 0 0 0 0>}
        cornerRadii=<0 0 0 0>
       }
       tintColor {colorUseRule=UseColor_Specified, specifiedColor=<rgba 1 1 1 1>}
       uVRegion {bIsValid=false, max=<0 0>, min=<0 0>}
      }
      horizontalBottomSlotImage {
       bIsDynamicallyLoaded=false
       drawAs=NoDrawType
       imageType=NoImage
       mirroring=NoMirror
       resourceName=None
       resourceObject=null
       tiling=NoTile
       imageSize=<0 0>
       margin=<ltrb 0 0 0 0>
       outlineSettings {
        bUseBrushTransparency=false
        roundingType=HalfHeightRadius
        width=0
        color {colorUseRule=UseColor_Specified, specifiedColor=<rgba 0 0 0 0>}
        cornerRadii=<0 0 0 0>
       }
       tintColor {colorUseRule=UseColor_Specified, specifiedColor=<rgba 1 1 1 1>}
       uVRegion {bIsValid=false, max=<0 0>, min=<0 0>}
      }
      horizontalTopSlotImage {
       bIsDynamicallyLoaded=false
       drawAs=NoDrawType
       imageType=NoImage
       mirroring=NoMirror
       resourceName=None
       resourceObject=null
       tiling=NoTile
       imageSize=<0 0>
       margin=<ltrb 0 0 0 0>
       outlineSettings {
        bUseBrushTransparency=false
        roundingType=HalfHeightRadius
        width=0
        color {colorUseRule=UseColor_Specified, specifiedColor=<rgba 0 0 0 0>}
        cornerRadii=<0 0 0 0>
       }
       tintColor {colorUseRule=UseColor_Specified, specifiedColor=<rgba 1 1 1 1>}
       uVRegion {bIsValid=false, max=<0 0>, min=<0 0>}
      }
      hoveredThumbImage {
       bIsDynamicallyLoaded=false
       drawAs=RoundedBox
       imageType=NoImage
       mirroring=NoMirror
       resourceName=None
       resourceObject=null
       tiling=NoTile
       imageSize=<0 0>
       margin=<ltrb 0 0 0 0>
       outlineSettings {
        bUseBrushTransparency=false
        roundingType=FixedRadius
        width=0
        color {colorUseRule=UseColor_Specified, specifiedColor=<rgba 0 0 0 0>}
        cornerRadii=<4 4 4 4>
       }
       tintColor {colorUseRule=UseColor_Specified, specifiedColor=<rgba 0.2159 0.2159 0.2159 1>}
       uVRegion {bIsValid=false, max=<0 0>, min=<0 0>}
      }
      normalThumbImage {
       bIsDynamicallyLoaded=false
       drawAs=RoundedBox
       imageType=NoImage
       mirroring=NoMirror
       resourceName=None
       resourceObject=null
       tiling=NoTile
       imageSize=<0 0>
       margin=<ltrb 0 0 0 0>
       outlineSettings {
        bUseBrushTransparency=false
        roundingType=FixedRadius
        width=0
        color {colorUseRule=UseColor_Specified, specifiedColor=<rgba 0 0 0 0>}
        cornerRadii=<4 4 4 4>
       }
       tintColor {colorUseRule=UseColor_Specified, specifiedColor=<rgba 0.0953 0.0953 0.0953 1>}
       uVRegion {bIsValid=false, max=<0 0>, min=<0 0>}
      }
      verticalBackgroundImage {
       bIsDynamicallyLoaded=false
       drawAs=NoDrawType
       imageType=NoImage
       mirroring=NoMirror
       resourceName=None
       resourceObject=null
       tiling=NoTile
       imageSize=<0 0>
       margin=<ltrb 0 0 0 0>
       outlineSettings {
        bUseBrushTransparency=false
        roundingType=HalfHeightRadius
        width=0
        color {colorUseRule=UseColor_Specified, specifiedColor=<rgba 0 0 0 0>}
        cornerRadii=<0 0 0 0>
       }
       tintColor {colorUseRule=UseColor_Specified, specifiedColor=<rgba 1 1 1 1>}
       uVRegion {bIsValid=false, max=<0 0>, min=<0 0>}
      }
      verticalBottomSlotImage {
       bIsDynamicallyLoaded=false
       drawAs=NoDrawType
       imageType=NoImage
       mirroring=NoMirror
       resourceName=None
       resourceObject=null
       tiling=NoTile
       imageSize=<0 0>
       margin=<ltrb 0 0 0 0>
       outlineSettings {
        bUseBrushTransparency=false
        roundingType=HalfHeightRadius
        width=0
        color {colorUseRule=UseColor_Specified, specifiedColor=<rgba 0 0 0 0>}
        cornerRadii=<0 0 0 0>
       }
       tintColor {colorUseRule=UseColor_Specified, specifiedColor=<rgba 1 1 1 1>}
       uVRegion {bIsValid=false, max=<0 0>, min=<0 0>}
      }
      verticalTopSlotImage {
       bIsDynamicallyLoaded=false
       drawAs=NoDrawType
       imageType=NoImage
       mirroring=NoMirror
       resourceName=None
       resourceObject=null
       tiling=NoTile
       imageSize=<0 0>
       margin=<ltrb 0 0 0 0>
       outlineSettings {
        bUseBrushTransparency=false
        roundingType=HalfHeightRadius
        width=0
        color {colorUseRule=UseColor_Specified, specifiedColor=<rgba 0 0 0 0>}
        cornerRadii=<0 0 0 0>
       }
       tintColor {colorUseRule=UseColor_Specified, specifiedColor=<rgba 1 1 1 1>}
       uVRegion {bIsValid=false, max=<0 0>, min=<0 0>}
      }
     }
     textStyle {
      overflowPolicy=Clip
      transformPolicy=None
      colorAndOpacity {colorUseRule=UseColor_Foreground, specifiedColor=<rgba 1 0 1 1>}
      font {
       bForceMonospaced=false
       bMaterialIsStencil=false
       fontMaterial=null
       fontObject=null
       letterSpacing=0
       monospacedWidth=1
       size=10
       skewAmount=0
       typefaceFontName=Regular
       outlineSettings {
        bApplyOutlineToDropShadows=false
        bMiteredCorners=false
        bSeparateFillAlpha=false
        outlineMaterial=null
        outlineSize=0
        outlineColor=<rgba 0 0 0 1>
       }
      }
      highlightColor {colorUseRule=UseColor_ColorTable, specifiedColor=<rgba 0 0 0 0>}
      highlightShape {
       bIsDynamicallyLoaded=false
       drawAs=Image
       imageType=NoImage
       mirroring=NoMirror
       resourceName=None
       resourceObject=null
       tiling=NoTile
       imageSize=<0 0>
       margin=<ltrb 0 0 0 0>
       outlineSettings {
        bUseBrushTransparency=false
        roundingType=HalfHeightRadius
        width=0
        color {colorUseRule=UseColor_Specified, specifiedColor=<rgba 0 0 0 0>}
        cornerRadii=<0 0 0 0>
       }
       tintColor {colorUseRule=UseColor_ColorTable, specifiedColor=<rgba 0 0 0 0>}
       uVRegion {bIsValid=false, max=<0 0>, min=<0 0>}
      }
      selectedBackgroundColor {colorUseRule=UseColor_ColorTable, specifiedColor=<rgba 0 0 0 0>}
      shadowColorAndOpacity=<rgba 0 0 0 1>
      shadowOffset=<0 0>
      strikeBrush {
       bIsDynamicallyLoaded=false
       drawAs=Image
       imageType=NoImage
       mirroring=NoMirror
       resourceName=None
       resourceObject=null
       tiling=NoTile
       imageSize=<32 32>
       margin=<ltrb 0 0 0 0>
       outlineSettings {
        bUseBrushTransparency=false
        roundingType=HalfHeightRadius
        width=0
        color {colorUseRule=UseColor_Specified, specifiedColor=<rgba 0 0 0 0>}
        cornerRadii=<0 0 0 0>
       }
       tintColor {colorUseRule=UseColor_Specified, specifiedColor=<rgba 1 1 1 1>}
       uVRegion {bIsValid=false, max=<0 0>, min=<0 0>}
      }
      underlineBrush {
       bIsDynamicallyLoaded=false
       drawAs=Image
       imageType=NoImage
       mirroring=NoMirror
       resourceName=None
       resourceObject=null
       tiling=NoTile
       imageSize=<32 32>
       margin=<ltrb 0 0 0 0>
       outlineSettings {
        bUseBrushTransparency=false
        roundingType=HalfHeightRadius
        width=0
        color {colorUseRule=UseColor_Specified, specifiedColor=<rgba 0 0 0 0>}
        cornerRadii=<0 0 0 0>
       }
       tintColor {colorUseRule=UseColor_Specified, specifiedColor=<rgba 1 1 1 1>}
       uVRegion {bIsValid=false, max=<0 0>, min=<0 0>}
      }
     }
     vScrollBarPadding=<ltrb 0 0 0 0>
    }
   }
   slot {class=VerticalBoxSlot, properties {size {sizeRule=Fill, value=1}}}
  }
  VerticalBox_189 {
   widgetClassPath=&VerticalBox
   parent=HbDataSource
   widget=VerticalBox_189
   slot {class=HorizontalBoxSlot, properties {size {sizeRule=Fill, value=1}}}
  }
  btnLoadedLevel {
   widgetClassPath=&TA_Generic01_Button
   parent=VerticalBox_189
   widget=btnLoadedLevel
   properties {label=Loaded Level Instances, visibility=Visible, buttonColor_Default=<rgba 1 0 0.0184 1>}
   slot {class=VerticalBoxSlot, properties {padding=<ltrb 5 0 5 0>}}
  }
  txtLoadedLevel {
   widgetClassPath=&EditorUtilityMultiLineEditableTextBox
   bIsVariable=true
   parent=VerticalBox_189
   widget=txtLoadedLevel
   properties {
    bIsReadOnly=true
    justification=Center
    text="Class usages will be found from all level instances in the currently loaded level.  (This will not find usages unloaded in World Partition)\r\nLeft-Clicking entries will select the asset. \r\nRight-Clicking entries will select level instances. (cannot select instances inside of BuildingLevelInstances)"
    widgetStyle {
     backgroundColor {colorUseRule=UseColor_Specified, specifiedColor=<rgba 1 1 1 1>}
     backgroundImageFocused {
      bIsDynamicallyLoaded=false
      drawAs=RoundedBox
      imageType=NoImage
      mirroring=NoMirror
      resourceName=None
      resourceObject=null
      tiling=NoTile
      imageSize=<0 0>
      margin=<ltrb 0 0 0 0>
      outlineSettings {
       bUseBrushTransparency=false
       roundingType=FixedRadius
       width=1
       color {colorUseRule=UseColor_ColorTable, specifiedColor=<rgba 0 0 0 0>}
       cornerRadii=<4 4 4 4>
      }
      tintColor {colorUseRule=UseColor_Specified, specifiedColor=<rgba 0.0048 0.0048 0.0048 1>}
      uVRegion {bIsValid=false, max=<0 0>, min=<0 0>}
     }
     backgroundImageHovered {
      bIsDynamicallyLoaded=false
      drawAs=RoundedBox
      imageType=NoImage
      mirroring=NoMirror
      resourceName=None
      resourceObject=null
      tiling=NoTile
      imageSize=<0 0>
      margin=<ltrb 0 0 0 0>
      outlineSettings {
       bUseBrushTransparency=false
       roundingType=FixedRadius
       width=1
       color {colorUseRule=UseColor_ColorTable, specifiedColor=<rgba 0 0 0 0>}
       cornerRadii=<4 4 4 4>
      }
      tintColor {colorUseRule=UseColor_Specified, specifiedColor=<rgba 0.0048 0.0048 0.0048 1>}
      uVRegion {bIsValid=false, max=<0 0>, min=<0 0>}
     }
     backgroundImageNormal {
      bIsDynamicallyLoaded=false
      drawAs=RoundedBox
      imageType=NoImage
      mirroring=NoMirror
      resourceName=None
      resourceObject=null
      tiling=NoTile
      imageSize=<0 0>
      margin=<ltrb 0 0 0 0>
      outlineSettings {
       bUseBrushTransparency=false
       roundingType=FixedRadius
       width=1
       color {colorUseRule=UseColor_ColorTable, specifiedColor=<rgba 0 0 0 0>}
       cornerRadii=<4 4 4 4>
      }
      tintColor {colorUseRule=UseColor_Specified, specifiedColor=<rgba 0.0048 0.0048 0.0048 1>}
      uVRegion {bIsValid=false, max=<0 0>, min=<0 0>}
     }
     backgroundImageReadOnly {
      bIsDynamicallyLoaded=false
      drawAs=RoundedBox
      imageType=NoImage
      mirroring=NoMirror
      resourceName=None
      resourceObject=null
      tiling=NoTile
      imageSize=<0 0>
      margin=<ltrb 0 0 0 0>
      outlineSettings {
       bUseBrushTransparency=false
       roundingType=FixedRadius
       width=1
       color {colorUseRule=UseColor_ColorTable, specifiedColor=<rgba 0 0 0 0>}
       cornerRadii=<4 4 4 4>
      }
      tintColor {colorUseRule=UseColor_Specified, specifiedColor=<rgba 0 0 0 1>}
      uVRegion {bIsValid=false, max=<0 0>, min=<0 0>}
     }
     focusedForegroundColor {colorUseRule=UseColor_Specified, specifiedColor=<rgba 1 1 1 1>}
     foregroundColor {colorUseRule=UseColor_Specified, specifiedColor=<rgba 0.5271 0.5271 0.5271 1>}
     hScrollBarPadding=<ltrb 0 0 0 0>
     padding=<ltrb 8 3 8 4>
     readOnlyForegroundColor {colorUseRule=UseColor_Specified, specifiedColor=<rgba 1 1 1 1>}
     scrollBarStyle {
      thickness=8
      draggedThumbImage {
       bIsDynamicallyLoaded=false
       drawAs=RoundedBox
       imageType=NoImage
       mirroring=NoMirror
       resourceName=None
       resourceObject=null
       tiling=NoTile
       imageSize=<0 0>
       margin=<ltrb 0 0 0 0>
       outlineSettings {
        bUseBrushTransparency=false
        roundingType=FixedRadius
        width=0
        color {colorUseRule=UseColor_Specified, specifiedColor=<rgba 0 0 0 0>}
        cornerRadii=<4 4 4 4>
       }
       tintColor {colorUseRule=UseColor_Specified, specifiedColor=<rgba 0.2159 0.2159 0.2159 1>}
       uVRegion {bIsValid=false, max=<0 0>, min=<0 0>}
      }
      horizontalBackgroundImage {
       bIsDynamicallyLoaded=false
       drawAs=NoDrawType
       imageType=NoImage
       mirroring=NoMirror
       resourceName=None
       resourceObject=null
       tiling=NoTile
       imageSize=<0 0>
       margin=<ltrb 0 0 0 0>
       outlineSettings {
        bUseBrushTransparency=false
        roundingType=HalfHeightRadius
        width=0
        color {colorUseRule=UseColor_Specified, specifiedColor=<rgba 0 0 0 0>}
        cornerRadii=<0 0 0 0>
       }
       tintColor {colorUseRule=UseColor_Specified, specifiedColor=<rgba 1 1 1 1>}
       uVRegion {bIsValid=false, max=<0 0>, min=<0 0>}
      }
      horizontalBottomSlotImage {
       bIsDynamicallyLoaded=false
       drawAs=NoDrawType
       imageType=NoImage
       mirroring=NoMirror
       resourceName=None
       resourceObject=null
       tiling=NoTile
       imageSize=<0 0>
       margin=<ltrb 0 0 0 0>
       outlineSettings {
        bUseBrushTransparency=false
        roundingType=HalfHeightRadius
        width=0
        color {colorUseRule=UseColor_Specified, specifiedColor=<rgba 0 0 0 0>}
        cornerRadii=<0 0 0 0>
       }
       tintColor {colorUseRule=UseColor_Specified, specifiedColor=<rgba 1 1 1 1>}
       uVRegion {bIsValid=false, max=<0 0>, min=<0 0>}
      }
      horizontalTopSlotImage {
       bIsDynamicallyLoaded=false
       drawAs=NoDrawType
       imageType=NoImage
       mirroring=NoMirror
       resourceName=None
       resourceObject=null
       tiling=NoTile
       imageSize=<0 0>
       margin=<ltrb 0 0 0 0>
       outlineSettings {
        bUseBrushTransparency=false
        roundingType=HalfHeightRadius
        width=0
        color {colorUseRule=UseColor_Specified, specifiedColor=<rgba 0 0 0 0>}
        cornerRadii=<0 0 0 0>
       }
       tintColor {colorUseRule=UseColor_Specified, specifiedColor=<rgba 1 1 1 1>}
       uVRegion {bIsValid=false, max=<0 0>, min=<0 0>}
      }
      hoveredThumbImage {
       bIsDynamicallyLoaded=false
       drawAs=RoundedBox
       imageType=NoImage
       mirroring=NoMirror
       resourceName=None
       resourceObject=null
       tiling=NoTile
       imageSize=<0 0>
       margin=<ltrb 0 0 0 0>
       outlineSettings {
        bUseBrushTransparency=false
        roundingType=FixedRadius
        width=0
        color {colorUseRule=UseColor_Specified, specifiedColor=<rgba 0 0 0 0>}
        cornerRadii=<4 4 4 4>
       }
       tintColor {colorUseRule=UseColor_Specified, specifiedColor=<rgba 0.2159 0.2159 0.2159 1>}
       uVRegion {bIsValid=false, max=<0 0>, min=<0 0>}
      }
      normalThumbImage {
       bIsDynamicallyLoaded=false
       drawAs=RoundedBox
       imageType=NoImage
       mirroring=NoMirror
       resourceName=None
       resourceObject=null
       tiling=NoTile
       imageSize=<0 0>
       margin=<ltrb 0 0 0 0>
       outlineSettings {
        bUseBrushTransparency=false
        roundingType=FixedRadius
        width=0
        color {colorUseRule=UseColor_Specified, specifiedColor=<rgba 0 0 0 0>}
        cornerRadii=<4 4 4 4>
       }
       tintColor {colorUseRule=UseColor_Specified, specifiedColor=<rgba 0.0953 0.0953 0.0953 1>}
       uVRegion {bIsValid=false, max=<0 0>, min=<0 0>}
      }
      verticalBackgroundImage {
       bIsDynamicallyLoaded=false
       drawAs=NoDrawType
       imageType=NoImage
       mirroring=NoMirror
       resourceName=None
       resourceObject=null
       tiling=NoTile
       imageSize=<0 0>
       margin=<ltrb 0 0 0 0>
       outlineSettings {
        bUseBrushTransparency=false
        roundingType=HalfHeightRadius
        width=0
        color {colorUseRule=UseColor_Specified, specifiedColor=<rgba 0 0 0 0>}
        cornerRadii=<0 0 0 0>
       }
       tintColor {colorUseRule=UseColor_Specified, specifiedColor=<rgba 1 1 1 1>}
       uVRegion {bIsValid=false, max=<0 0>, min=<0 0>}
      }
      verticalBottomSlotImage {
       bIsDynamicallyLoaded=false
       drawAs=NoDrawType
       imageType=NoImage
       mirroring=NoMirror
       resourceName=None
       resourceObject=null
       tiling=NoTile
       imageSize=<0 0>
       margin=<ltrb 0 0 0 0>
       outlineSettings {
        bUseBrushTransparency=false
        roundingType=HalfHeightRadius
        width=0
        color {colorUseRule=UseColor_Specified, specifiedColor=<rgba 0 0 0 0>}
        cornerRadii=<0 0 0 0>
       }
       tintColor {colorUseRule=UseColor_Specified, specifiedColor=<rgba 1 1 1 1>}
       uVRegion {bIsValid=false, max=<0 0>, min=<0 0>}
      }
      verticalTopSlotImage {
       bIsDynamicallyLoaded=false
       drawAs=NoDrawType
       imageType=NoImage
       mirroring=NoMirror
       resourceName=None
       resourceObject=null
       tiling=NoTile
       imageSize=<0 0>
       margin=<ltrb 0 0 0 0>
       outlineSettings {
        bUseBrushTransparency=false
        roundingType=HalfHeightRadius
        width=0
        color {colorUseRule=UseColor_Specified, specifiedColor=<rgba 0 0 0 0>}
        cornerRadii=<0 0 0 0>
       }
       tintColor {colorUseRule=UseColor_Specified, specifiedColor=<rgba 1 1 1 1>}
       uVRegion {bIsValid=false, max=<0 0>, min=<0 0>}
      }
     }
     textStyle {
      overflowPolicy=Clip
      transformPolicy=None
      colorAndOpacity {colorUseRule=UseColor_Foreground, specifiedColor=<rgba 1 0 1 1>}
      font {
       bForceMonospaced=false
       bMaterialIsStencil=false
       fontMaterial=null
       fontObject=null
       letterSpacing=0
       monospacedWidth=1
       size=10
       skewAmount=0
       typefaceFontName=Regular
       outlineSettings {
        bApplyOutlineToDropShadows=false
        bMiteredCorners=false
        bSeparateFillAlpha=false
        outlineMaterial=null
        outlineSize=0
        outlineColor=<rgba 0 0 0 1>
       }
      }
      highlightColor {colorUseRule=UseColor_ColorTable, specifiedColor=<rgba 0 0 0 0>}
      highlightShape {
       bIsDynamicallyLoaded=false
       drawAs=Image
       imageType=NoImage
       mirroring=NoMirror
       resourceName=None
       resourceObject=null
       tiling=NoTile
       imageSize=<0 0>
       margin=<ltrb 0 0 0 0>
       outlineSettings {
        bUseBrushTransparency=false
        roundingType=HalfHeightRadius
        width=0
        color {colorUseRule=UseColor_Specified, specifiedColor=<rgba 0 0 0 0>}
        cornerRadii=<0 0 0 0>
       }
       tintColor {colorUseRule=UseColor_ColorTable, specifiedColor=<rgba 0 0 0 0>}
       uVRegion {bIsValid=false, max=<0 0>, min=<0 0>}
      }
      selectedBackgroundColor {colorUseRule=UseColor_ColorTable, specifiedColor=<rgba 0 0 0 0>}
      shadowColorAndOpacity=<rgba 0 0 0 1>
      shadowOffset=<0 0>
      strikeBrush {
       bIsDynamicallyLoaded=false
       drawAs=Image
       imageType=NoImage
       mirroring=NoMirror
       resourceName=None
       resourceObject=null
       tiling=NoTile
       imageSize=<32 32>
       margin=<ltrb 0 0 0 0>
       outlineSettings {
        bUseBrushTransparency=false
        roundingType=HalfHeightRadius
        width=0
        color {colorUseRule=UseColor_Specified, specifiedColor=<rgba 0 0 0 0>}
        cornerRadii=<0 0 0 0>
       }
       tintColor {colorUseRule=UseColor_Specified, specifiedColor=<rgba 1 1 1 1>}
       uVRegion {bIsValid=false, max=<0 0>, min=<0 0>}
      }
      underlineBrush {
       bIsDynamicallyLoaded=false
       drawAs=Image
       imageType=NoImage
       mirroring=NoMirror
       resourceName=None
       resourceObject=null
       tiling=NoTile
       imageSize=<32 32>
       margin=<ltrb 0 0 0 0>
       outlineSettings {
        bUseBrushTransparency=false
        roundingType=HalfHeightRadius
        width=0
        color {colorUseRule=UseColor_Specified, specifiedColor=<rgba 0 0 0 0>}
        cornerRadii=<0 0 0 0>
       }
       tintColor {colorUseRule=UseColor_Specified, specifiedColor=<rgba 1 1 1 1>}
       uVRegion {bIsValid=false, max=<0 0>, min=<0 0>}
      }
     }
     vScrollBarPadding=<ltrb 0 0 0 0>
    }
   }
   slot {class=VerticalBoxSlot, properties {size {sizeRule=Fill, value=2}}}
  }
  HbTargetSetup {widgetClassPath=&HorizontalBox, bIsVariable=true, parent=VerticalBox_0, widget=HbTargetSetup, slot {class=VerticalBoxSlot}}
  PropTargetClass {
   widgetClassPath=&SinglePropertyView
   parent=HbTargetSetup
   widget=PropTargetClass
   slot {class=HorizontalBoxSlot, properties {size {sizeRule=Fill, value=1}}}
  }
  PropTargetClassPath {
   widgetClassPath=&SinglePropertyView
   parent=HbTargetSetup
   widget=PropTargetClassPath
   properties {visibility=Collapsed}
   slot {class=HorizontalBoxSlot, properties {size {sizeRule=Fill, value=1}}}
  }
  PropIncludeNativeClasses {
   widgetClassPath=&SinglePropertyView
   parent=HbTargetSetup
   widget=PropIncludeNativeClasses
   slot {class=HorizontalBoxSlot}
  }
  btnAdvanced {
   widgetClassPath=&TA_Generic01_Button
   parent=HbTargetSetup
   widget=btnAdvanced
   properties {
    label=Advanced
    toolTipText="This switches between Basic class selection using the selection dialog, and advanced where you can enter the path of the class."
    visibility=Visible
   }
   slot {class=HorizontalBoxSlot, properties {padding=<ltrb 5 0 5 0>}}
  }
  HbFilterSetup {widgetClassPath=&HorizontalBox, bIsVariable=true, parent=VerticalBox_0, widget=HbFilterSetup, slot {class=VerticalBoxSlot}}
  txtCount {
   widgetClassPath=&EditableTextBox
   bIsVariable=true
   parent=HbFilterSetup
   widget=txtCount
   properties {
    isReadOnly=true
    justification=Center
    minimumDesiredWidth=25.0596
    text=???
    visibility=SelfHitTestInvisible
    widgetStyle {
     backgroundColor {colorUseRule=UseColor_Specified, specifiedColor=<rgba 0.1 0.1 0.1 1>}
     backgroundImageFocused {
      bIsDynamicallyLoaded=false
      drawAs=RoundedBox
      imageType=NoImage
      mirroring=NoMirror
      resourceName=None
      resourceObject=null
      tiling=NoTile
      imageSize=<0 0>
      margin=<ltrb 0 0 0 0>
      outlineSettings {
       bUseBrushTransparency=false
       roundingType=FixedRadius
       width=1
       color {colorUseRule=UseColor_ColorTable, specifiedColor=<rgba 0 0 0 0>}
       cornerRadii=<4 4 4 4>
      }
      tintColor {colorUseRule=UseColor_Specified, specifiedColor=<rgba 0.0048 0.0048 0.0048 1>}
      uVRegion {bIsValid=false, max=<0 0>, min=<0 0>}
     }
     backgroundImageHovered {
      bIsDynamicallyLoaded=false
      drawAs=RoundedBox
      imageType=NoImage
      mirroring=NoMirror
      resourceName=None
      resourceObject=null
      tiling=NoTile
      imageSize=<0 0>
      margin=<ltrb 0 0 0 0>
      outlineSettings {
       bUseBrushTransparency=false
       roundingType=FixedRadius
       width=1
       color {colorUseRule=UseColor_ColorTable, specifiedColor=<rgba 0 0 0 0>}
       cornerRadii=<4 4 4 4>
      }
      tintColor {colorUseRule=UseColor_Specified, specifiedColor=<rgba 0.0048 0.0048 0.0048 1>}
      uVRegion {bIsValid=false, max=<0 0>, min=<0 0>}
     }
     backgroundImageNormal {
      bIsDynamicallyLoaded=false
      drawAs=RoundedBox
      imageType=NoImage
      mirroring=NoMirror
      resourceName=None
      resourceObject=null
      tiling=NoTile
      imageSize=<0 0>
      margin=<ltrb 0 0 0 0>
      outlineSettings {
       bUseBrushTransparency=false
       roundingType=FixedRadius
       width=1
       color {colorUseRule=UseColor_ColorTable, specifiedColor=<rgba 0 0 0 0>}
       cornerRadii=<4 4 4 4>
      }
      tintColor {colorUseRule=UseColor_Specified, specifiedColor=<rgba 0.0048 0.0048 0.0048 1>}
      uVRegion {bIsValid=false, max=<0 0>, min=<0 0>}
     }
     backgroundImageReadOnly {
      bIsDynamicallyLoaded=false
      drawAs=RoundedBox
      imageType=NoImage
      mirroring=NoMirror
      resourceName=None
      resourceObject=null
      tiling=NoTile
      imageSize=<0 0>
      margin=<ltrb 0 0 0 0>
      outlineSettings {
       bUseBrushTransparency=false
       roundingType=FixedRadius
       width=1
       color {colorUseRule=UseColor_ColorTable, specifiedColor=<rgba 0 0 0 0>}
       cornerRadii=<4 4 4 4>
      }
      tintColor {colorUseRule=UseColor_Specified, specifiedColor=<rgba 0.0284 0.0284 0.0284 1>}
      uVRegion {bIsValid=false, max=<0 0>, min=<0 0>}
     }
     focusedForegroundColor {colorUseRule=UseColor_Specified, specifiedColor=<rgba 1 1 1 1>}
     foregroundColor {colorUseRule=UseColor_Specified, specifiedColor=<rgba 1 1 1 1>}
     hScrollBarPadding=<ltrb 0 0 0 0>
     padding=<ltrb 8 3 8 4>
     readOnlyForegroundColor {colorUseRule=UseColor_Specified, specifiedColor=<rgba 1 1 1 1>}
     scrollBarStyle {
      thickness=8
      draggedThumbImage {
       bIsDynamicallyLoaded=false
       drawAs=RoundedBox
       imageType=NoImage
       mirroring=NoMirror
       resourceName=None
       resourceObject=null
       tiling=NoTile
       imageSize=<0 0>
       margin=<ltrb 0 0 0 0>
       outlineSettings {
        bUseBrushTransparency=false
        roundingType=FixedRadius
        width=0
        color {colorUseRule=UseColor_Specified, specifiedColor=<rgba 0 0 0 0>}
        cornerRadii=<4 4 4 4>
       }
       tintColor {colorUseRule=UseColor_Specified, specifiedColor=<rgba 0.2159 0.2159 0.2159 1>}
       uVRegion {bIsValid=false, max=<0 0>, min=<0 0>}
      }
      horizontalBackgroundImage {
       bIsDynamicallyLoaded=false
       drawAs=NoDrawType
       imageType=NoImage
       mirroring=NoMirror
       resourceName=None
       resourceObject=null
       tiling=NoTile
       imageSize=<0 0>
       margin=<ltrb 0 0 0 0>
       outlineSettings {
        bUseBrushTransparency=false
        roundingType=HalfHeightRadius
        width=0
        color {colorUseRule=UseColor_Specified, specifiedColor=<rgba 0 0 0 0>}
        cornerRadii=<0 0 0 0>
       }
       tintColor {colorUseRule=UseColor_Specified, specifiedColor=<rgba 1 1 1 1>}
       uVRegion {bIsValid=false, max=<0 0>, min=<0 0>}
      }
      horizontalBottomSlotImage {
       bIsDynamicallyLoaded=false
       drawAs=NoDrawType
       imageType=NoImage
       mirroring=NoMirror
       resourceName=None
       resourceObject=null
       tiling=NoTile
       imageSize=<0 0>
       margin=<ltrb 0 0 0 0>
       outlineSettings {
        bUseBrushTransparency=false
        roundingType=HalfHeightRadius
        width=0
        color {colorUseRule=UseColor_Specified, specifiedColor=<rgba 0 0 0 0>}
        cornerRadii=<0 0 0 0>
       }
       tintColor {colorUseRule=UseColor_Specified, specifiedColor=<rgba 1 1 1 1>}
       uVRegion {bIsValid=false, max=<0 0>, min=<0 0>}
      }
      horizontalTopSlotImage {
       bIsDynamicallyLoaded=false
       drawAs=NoDrawType
       imageType=NoImage
       mirroring=NoMirror
       resourceName=None
       resourceObject=null
       tiling=NoTile
       imageSize=<0 0>
       margin=<ltrb 0 0 0 0>
       outlineSettings {
        bUseBrushTransparency=false
        roundingType=HalfHeightRadius
        width=0
        color {colorUseRule=UseColor_Specified, specifiedColor=<rgba 0 0 0 0>}
        cornerRadii=<0 0 0 0>
       }
       tintColor {colorUseRule=UseColor_Specified, specifiedColor=<rgba 1 1 1 1>}
       uVRegion {bIsValid=false, max=<0 0>, min=<0 0>}
      }
      hoveredThumbImage {
       bIsDynamicallyLoaded=false
       drawAs=RoundedBox
       imageType=NoImage
       mirroring=NoMirror
       resourceName=None
       resourceObject=null
       tiling=NoTile
       imageSize=<0 0>
       margin=<ltrb 0 0 0 0>
       outlineSettings {
        bUseBrushTransparency=false
        roundingType=FixedRadius
        width=0
        color {colorUseRule=UseColor_Specified, specifiedColor=<rgba 0 0 0 0>}
        cornerRadii=<4 4 4 4>
       }
       tintColor {colorUseRule=UseColor_Specified, specifiedColor=<rgba 0.2159 0.2159 0.2159 1>}
       uVRegion {bIsValid=false, max=<0 0>, min=<0 0>}
      }
      normalThumbImage {
       bIsDynamicallyLoaded=false
       drawAs=RoundedBox
       imageType=NoImage
       mirroring=NoMirror
       resourceName=None
       resourceObject=null
       tiling=NoTile
       imageSize=<0 0>
       margin=<ltrb 0 0 0 0>
       outlineSettings {
        bUseBrushTransparency=false
        roundingType=FixedRadius
        width=0
        color {colorUseRule=UseColor_Specified, specifiedColor=<rgba 0 0 0 0>}
        cornerRadii=<4 4 4 4>
       }
       tintColor {colorUseRule=UseColor_Specified, specifiedColor=<rgba 0.0953 0.0953 0.0953 1>}
       uVRegion {bIsValid=false, max=<0 0>, min=<0 0>}
      }
      verticalBackgroundImage {
       bIsDynamicallyLoaded=false
       drawAs=NoDrawType
       imageType=NoImage
       mirroring=NoMirror
       resourceName=None
       resourceObject=null
       tiling=NoTile
       imageSize=<0 0>
       margin=<ltrb 0 0 0 0>
       outlineSettings {
        bUseBrushTransparency=false
        roundingType=HalfHeightRadius
        width=0
        color {colorUseRule=UseColor_Specified, specifiedColor=<rgba 0 0 0 0>}
        cornerRadii=<0 0 0 0>
       }
       tintColor {colorUseRule=UseColor_Specified, specifiedColor=<rgba 1 1 1 1>}
       uVRegion {bIsValid=false, max=<0 0>, min=<0 0>}
      }
      verticalBottomSlotImage {
       bIsDynamicallyLoaded=false
       drawAs=NoDrawType
       imageType=NoImage
       mirroring=NoMirror
       resourceName=None
       resourceObject=null
       tiling=NoTile
       imageSize=<0 0>
       margin=<ltrb 0 0 0 0>
       outlineSettings {
        bUseBrushTransparency=false
        roundingType=HalfHeightRadius
        width=0
        color {colorUseRule=UseColor_Specified, specifiedColor=<rgba 0 0 0 0>}
        cornerRadii=<0 0 0 0>
       }
       tintColor {colorUseRule=UseColor_Specified, specifiedColor=<rgba 1 1 1 1>}
       uVRegion {bIsValid=false, max=<0 0>, min=<0 0>}
      }
      verticalTopSlotImage {
       bIsDynamicallyLoaded=false
       drawAs=NoDrawType
       imageType=NoImage
       mirroring=NoMirror
       resourceName=None
       resourceObject=null
       tiling=NoTile
       imageSize=<0 0>
       margin=<ltrb 0 0 0 0>
       outlineSettings {
        bUseBrushTransparency=false
        roundingType=HalfHeightRadius
        width=0
        color {colorUseRule=UseColor_Specified, specifiedColor=<rgba 0 0 0 0>}
        cornerRadii=<0 0 0 0>
       }
       tintColor {colorUseRule=UseColor_Specified, specifiedColor=<rgba 1 1 1 1>}
       uVRegion {bIsValid=false, max=<0 0>, min=<0 0>}
      }
     }
     textStyle {
      overflowPolicy=Clip
      transformPolicy=None
      colorAndOpacity {colorUseRule=UseColor_Foreground, specifiedColor=<rgba 1 0 1 1>}
      font {
       bForceMonospaced=false
       bMaterialIsStencil=false
       fontMaterial=null
       fontObject=null
       letterSpacing=0
       monospacedWidth=1
       size=9
       skewAmount=0
       typefaceFontName=Regular
       outlineSettings {
        bApplyOutlineToDropShadows=false
        bMiteredCorners=false
        bSeparateFillAlpha=false
        outlineMaterial=null
        outlineSize=0
        outlineColor=<rgba 0 0 0 1>
       }
      }
      highlightColor {colorUseRule=UseColor_ColorTable, specifiedColor=<rgba 0 0 0 0>}
      highlightShape {
       bIsDynamicallyLoaded=false
       drawAs=Image
       imageType=NoImage
       mirroring=NoMirror
       resourceName=None
       resourceObject=null
       tiling=NoTile
       imageSize=<0 0>
       margin=<ltrb 0 0 0 0>
       outlineSettings {
        bUseBrushTransparency=false
        roundingType=HalfHeightRadius
        width=0
        color {colorUseRule=UseColor_Specified, specifiedColor=<rgba 0 0 0 0>}
        cornerRadii=<0 0 0 0>
       }
       tintColor {colorUseRule=UseColor_ColorTable, specifiedColor=<rgba 0 0 0 0>}
       uVRegion {bIsValid=false, max=<0 0>, min=<0 0>}
      }
      selectedBackgroundColor {colorUseRule=UseColor_ColorTable, specifiedColor=<rgba 0 0 0 0>}
      shadowColorAndOpacity=<rgba 0 0 0 1>
      shadowOffset=<0 0>
      strikeBrush {
       bIsDynamicallyLoaded=false
       drawAs=Image
       imageType=NoImage
       mirroring=NoMirror
       resourceName=None
       resourceObject=null
       tiling=NoTile
       imageSize=<32 32>
       margin=<ltrb 0 0 0 0>
       outlineSettings {
        bUseBrushTransparency=false
        roundingType=HalfHeightRadius
        width=0
        color {colorUseRule=UseColor_Specified, specifiedColor=<rgba 0 0 0 0>}
        cornerRadii=<0 0 0 0>
       }
       tintColor {colorUseRule=UseColor_Specified, specifiedColor=<rgba 1 1 1 1>}
       uVRegion {bIsValid=false, max=<0 0>, min=<0 0>}
      }
      underlineBrush {
       bIsDynamicallyLoaded=false
       drawAs=Image
       imageType=NoImage
       mirroring=NoMirror
       resourceName=None
       resourceObject=null
       tiling=NoTile
       imageSize=<32 32>
       margin=<ltrb 0 0 0 0>
       outlineSettings {
        bUseBrushTransparency=false
        roundingType=HalfHeightRadius
        width=0
        color {colorUseRule=UseColor_Specified, specifiedColor=<rgba 0 0 0 0>}
        cornerRadii=<0 0 0 0>
       }
       tintColor {colorUseRule=UseColor_Specified, specifiedColor=<rgba 1 1 1 1>}
       uVRegion {bIsValid=false, max=<0 0>, min=<0 0>}
      }
     }
     vScrollBarPadding=<ltrb 0 0 0 0>
    }
   }
   slot {class=HorizontalBoxSlot}
  }
  PropRequiredFilter {
   widgetClassPath=&SinglePropertyView
   parent=HbFilterSetup
   widget=PropRequiredFilter
   slot {class=HorizontalBoxSlot, properties {size {sizeRule=Fill, value=1}}}
  }
  PropExcludedFilter {
   widgetClassPath=&SinglePropertyView
   parent=HbFilterSetup
   widget=PropExcludedFilter
   slot {class=HorizontalBoxSlot, properties {size {sizeRule=Fill, value=1}}}
  }
  btnApplyCustomFilter {
   widgetClassPath=&TA_Generic01_Button
   parent=HbFilterSetup
   widget=btnApplyCustomFilter
   properties {
    label=Apply Custom Filter
    toolTipText="This will apply a custom filter to all the assets currently in the list. \r\n\r\nTo override this functionality, you can use the \"Select Tool BP\" button to edit the blueprint function \"Apply Custom Filter\".  \r\n\r\nAs well, if you have code access you can update \"UFortniteEditorLibrary::RunCustomFilterForFindClassUsages\" for custom code logic."
    visibility=Visible
    buttonColor_Default=<rgba 0.02 0.4 0 1>
   }
   slot {class=HorizontalBoxSlot, properties {padding=<ltrb 5 0 5 0>}}
  }
  btnOpenToolBP {
   widgetClassPath=&TA_Generic01_Button
   parent=HbFilterSetup
   widget=btnOpenToolBP
   properties {
    label=Select Tool BP
    toolTipText="This will select the Blueprint asset for the Find Class Usages editor utility tool.  This is useful if you need to modify the custom filter."
    visibility=Visible
   }
   slot {class=HorizontalBoxSlot, properties {padding=<ltrb 5 0 5 0>}}
  }
  Border_0 {
   widgetClassPath=&Border
   parent=VerticalBox_0
   widget=Border_0
   properties {brushColor=<rgba 1 1 1 0.15>, padding=<ltrb 0 0 0 0>}
   slot {class=VerticalBoxSlot, properties {padding=<ltrb 0 5 0 0>, size {sizeRule=Fill, value=1}}}
  }
  VerticalBox_224 {
   widgetClassPath=&VerticalBox
   parent=Border_0
   widget=VerticalBox_224
   slot {class=BorderSlot, properties {padding=<ltrb 0 0 0 0>}}
  }
  LstAssetPaths {
   widgetClassPath=/Game/Blueprints/Blutilities/ForTechArt/Widgets/List/TA_Generic01_List
   bIsVariable=true
   parent=VerticalBox_224
   widget=LstAssetPaths
   properties {allowMultiSelect=false, allowNoSelection=true}
   slot {class=VerticalBoxSlot, properties {size {sizeRule=Fill, value=1}}}
  }
  HorizontalBox_1 {
   widgetClassPath=&HorizontalBox
   parent=VerticalBox_224
   widget=HorizontalBox_1
   slot {class=VerticalBoxSlot, properties {horizontalAlignment=HAlign_Center}}
  }
  btnPageLeft {
   widgetClassPath=&TA_Generic01_Button
   parent=HorizontalBox_1
   widget=btnPageLeft
   properties {label="<<< Prev Page ", visibility=Visible, buttonColor_Default=<rgba 0 0.474 0.0147 1>, buttonColor_Selected=<rgba 0.6828 1 0.0254 1>}
   slot {class=HorizontalBoxSlot, properties {horizontalAlignment=HAlign_Center, verticalAlignment=VAlign_Center, padding=<ltrb 5 0 5 0>}}
  }
  txtPageCount {
   widgetClassPath=&EditableTextBox
   bIsVariable=true
   parent=HorizontalBox_1
   widget=txtPageCount
   properties {
    isReadOnly=true
    justification=Center
    minimumDesiredWidth=25.0596
    text=???
    visibility=SelfHitTestInvisible
    widgetStyle {
     backgroundColor {colorUseRule=UseColor_Specified, specifiedColor=<rgba 0.1 0.1 0.1 1>}
     backgroundImageFocused {
      bIsDynamicallyLoaded=false
      drawAs=RoundedBox
      imageType=NoImage
      mirroring=NoMirror
      resourceName=None
      resourceObject=null
      tiling=NoTile
      imageSize=<0 0>
      margin=<ltrb 0 0 0 0>
      outlineSettings {
       bUseBrushTransparency=false
       roundingType=FixedRadius
       width=1
       color {colorUseRule=UseColor_ColorTable, specifiedColor=<rgba 0 0 0 0>}
       cornerRadii=<4 4 4 4>
      }
      tintColor {colorUseRule=UseColor_Specified, specifiedColor=<rgba 0.0048 0.0048 0.0048 1>}
      uVRegion {bIsValid=false, max=<0 0>, min=<0 0>}
     }
     backgroundImageHovered {
      bIsDynamicallyLoaded=false
      drawAs=RoundedBox
      imageType=NoImage
      mirroring=NoMirror
      resourceName=None
      resourceObject=null
      tiling=NoTile
      imageSize=<0 0>
      margin=<ltrb 0 0 0 0>
      outlineSettings {
       bUseBrushTransparency=false
       roundingType=FixedRadius
       width=1
       color {colorUseRule=UseColor_ColorTable, specifiedColor=<rgba 0 0 0 0>}
       cornerRadii=<4 4 4 4>
      }
      tintColor {colorUseRule=UseColor_Specified, specifiedColor=<rgba 0.0048 0.0048 0.0048 1>}
      uVRegion {bIsValid=false, max=<0 0>, min=<0 0>}
     }
     backgroundImageNormal {
      bIsDynamicallyLoaded=false
      drawAs=RoundedBox
      imageType=NoImage
      mirroring=NoMirror
      resourceName=None
      resourceObject=null
      tiling=NoTile
      imageSize=<0 0>
      margin=<ltrb 0 0 0 0>
      outlineSettings {
       bUseBrushTransparency=false
       roundingType=FixedRadius
       width=1
       color {colorUseRule=UseColor_ColorTable, specifiedColor=<rgba 0 0 0 0>}
       cornerRadii=<4 4 4 4>
      }
      tintColor {colorUseRule=UseColor_Specified, specifiedColor=<rgba 0.0048 0.0048 0.0048 1>}
      uVRegion {bIsValid=false, max=<0 0>, min=<0 0>}
     }
     backgroundImageReadOnly {
      bIsDynamicallyLoaded=false
      drawAs=RoundedBox
      imageType=NoImage
      mirroring=NoMirror
      resourceName=None
      resourceObject=null
      tiling=NoTile
      imageSize=<0 0>
      margin=<ltrb 0 0 0 0>
      outlineSettings {
       bUseBrushTransparency=false
       roundingType=FixedRadius
       width=1
       color {colorUseRule=UseColor_ColorTable, specifiedColor=<rgba 0 0 0 0>}
       cornerRadii=<4 4 4 4>
      }
      tintColor {colorUseRule=UseColor_Specified, specifiedColor=<rgba 0.0284 0.0284 0.0284 1>}
      uVRegion {bIsValid=false, max=<0 0>, min=<0 0>}
     }
     focusedForegroundColor {colorUseRule=UseColor_Specified, specifiedColor=<rgba 1 1 1 1>}
     foregroundColor {colorUseRule=UseColor_Specified, specifiedColor=<rgba 1 1 1 1>}
     hScrollBarPadding=<ltrb 0 0 0 0>
     padding=<ltrb 8 3 8 4>
     readOnlyForegroundColor {colorUseRule=UseColor_Specified, specifiedColor=<rgba 1 1 1 1>}
     scrollBarStyle {
      thickness=8
      draggedThumbImage {
       bIsDynamicallyLoaded=false
       drawAs=RoundedBox
       imageType=NoImage
       mirroring=NoMirror
       resourceName=None
       resourceObject=null
       tiling=NoTile
       imageSize=<0 0>
       margin=<ltrb 0 0 0 0>
       outlineSettings {
        bUseBrushTransparency=false
        roundingType=FixedRadius
        width=0
        color {colorUseRule=UseColor_Specified, specifiedColor=<rgba 0 0 0 0>}
        cornerRadii=<4 4 4 4>
       }
       tintColor {colorUseRule=UseColor_Specified, specifiedColor=<rgba 0.2159 0.2159 0.2159 1>}
       uVRegion {bIsValid=false, max=<0 0>, min=<0 0>}
      }
      horizontalBackgroundImage {
       bIsDynamicallyLoaded=false
       drawAs=NoDrawType
       imageType=NoImage
       mirroring=NoMirror
       resourceName=None
       resourceObject=null
       tiling=NoTile
       imageSize=<0 0>
       margin=<ltrb 0 0 0 0>
       outlineSettings {
        bUseBrushTransparency=false
        roundingType=HalfHeightRadius
        width=0
        color {colorUseRule=UseColor_Specified, specifiedColor=<rgba 0 0 0 0>}
        cornerRadii=<0 0 0 0>
       }
       tintColor {colorUseRule=UseColor_Specified, specifiedColor=<rgba 1 1 1 1>}
       uVRegion {bIsValid=false, max=<0 0>, min=<0 0>}
      }
      horizontalBottomSlotImage {
       bIsDynamicallyLoaded=false
       drawAs=NoDrawType
       imageType=NoImage
       mirroring=NoMirror
       resourceName=None
       resourceObject=null
       tiling=NoTile
       imageSize=<0 0>
       margin=<ltrb 0 0 0 0>
       outlineSettings {
        bUseBrushTransparency=false
        roundingType=HalfHeightRadius
        width=0
        color {colorUseRule=UseColor_Specified, specifiedColor=<rgba 0 0 0 0>}
        cornerRadii=<0 0 0 0>
       }
       tintColor {colorUseRule=UseColor_Specified, specifiedColor=<rgba 1 1 1 1>}
       uVRegion {bIsValid=false, max=<0 0>, min=<0 0>}
      }
      horizontalTopSlotImage {
       bIsDynamicallyLoaded=false
       drawAs=NoDrawType
       imageType=NoImage
       mirroring=NoMirror
       resourceName=None
       resourceObject=null
       tiling=NoTile
       imageSize=<0 0>
       margin=<ltrb 0 0 0 0>
       outlineSettings {
        bUseBrushTransparency=false
        roundingType=HalfHeightRadius
        width=0
        color {colorUseRule=UseColor_Specified, specifiedColor=<rgba 0 0 0 0>}
        cornerRadii=<0 0 0 0>
       }
       tintColor {colorUseRule=UseColor_Specified, specifiedColor=<rgba 1 1 1 1>}
       uVRegion {bIsValid=false, max=<0 0>, min=<0 0>}
      }
      hoveredThumbImage {
       bIsDynamicallyLoaded=false
       drawAs=RoundedBox
       imageType=NoImage
       mirroring=NoMirror
       resourceName=None
       resourceObject=null
       tiling=NoTile
       imageSize=<0 0>
       margin=<ltrb 0 0 0 0>
       outlineSettings {
        bUseBrushTransparency=false
        roundingType=FixedRadius
        width=0
        color {colorUseRule=UseColor_Specified, specifiedColor=<rgba 0 0 0 0>}
        cornerRadii=<4 4 4 4>
       }
       tintColor {colorUseRule=UseColor_Specified, specifiedColor=<rgba 0.2159 0.2159 0.2159 1>}
       uVRegion {bIsValid=false, max=<0 0>, min=<0 0>}
      }
      normalThumbImage {
       bIsDynamicallyLoaded=false
       drawAs=RoundedBox
       imageType=NoImage
       mirroring=NoMirror
       resourceName=None
       resourceObject=null
       tiling=NoTile
       imageSize=<0 0>
       margin=<ltrb 0 0 0 0>
       outlineSettings {
        bUseBrushTransparency=false
        roundingType=FixedRadius
        width=0
        color {colorUseRule=UseColor_Specified, specifiedColor=<rgba 0 0 0 0>}
        cornerRadii=<4 4 4 4>
       }
       tintColor {colorUseRule=UseColor_Specified, specifiedColor=<rgba 0.0953 0.0953 0.0953 1>}
       uVRegion {bIsValid=false, max=<0 0>, min=<0 0>}
      }
      verticalBackgroundImage {
       bIsDynamicallyLoaded=false
       drawAs=NoDrawType
       imageType=NoImage
       mirroring=NoMirror
       resourceName=None
       resourceObject=null
       tiling=NoTile
       imageSize=<0 0>
       margin=<ltrb 0 0 0 0>
       outlineSettings {
        bUseBrushTransparency=false
        roundingType=HalfHeightRadius
        width=0
        color {colorUseRule=UseColor_Specified, specifiedColor=<rgba 0 0 0 0>}
        cornerRadii=<0 0 0 0>
       }
       tintColor {colorUseRule=UseColor_Specified, specifiedColor=<rgba 1 1 1 1>}
       uVRegion {bIsValid=false, max=<0 0>, min=<0 0>}
      }
      verticalBottomSlotImage {
       bIsDynamicallyLoaded=false
       drawAs=NoDrawType
       imageType=NoImage
       mirroring=NoMirror
       resourceName=None
       resourceObject=null
       tiling=NoTile
       imageSize=<0 0>
       margin=<ltrb 0 0 0 0>
       outlineSettings {
        bUseBrushTransparency=false
        roundingType=HalfHeightRadius
        width=0
        color {colorUseRule=UseColor_Specified, specifiedColor=<rgba 0 0 0 0>}
        cornerRadii=<0 0 0 0>
       }
       tintColor {colorUseRule=UseColor_Specified, specifiedColor=<rgba 1 1 1 1>}
       uVRegion {bIsValid=false, max=<0 0>, min=<0 0>}
      }
      verticalTopSlotImage {
       bIsDynamicallyLoaded=false
       drawAs=NoDrawType
       imageType=NoImage
       mirroring=NoMirror
       resourceName=None
       resourceObject=null
       tiling=NoTile
       imageSize=<0 0>
       margin=<ltrb 0 0 0 0>
       outlineSettings {
        bUseBrushTransparency=false
        roundingType=HalfHeightRadius
        width=0
        color {colorUseRule=UseColor_Specified, specifiedColor=<rgba 0 0 0 0>}
        cornerRadii=<0 0 0 0>
       }
       tintColor {colorUseRule=UseColor_Specified, specifiedColor=<rgba 1 1 1 1>}
       uVRegion {bIsValid=false, max=<0 0>, min=<0 0>}
      }
     }
     textStyle {
      overflowPolicy=Clip
      transformPolicy=None
      colorAndOpacity {colorUseRule=UseColor_Foreground, specifiedColor=<rgba 1 0 1 1>}
      font {
       bForceMonospaced=false
       bMaterialIsStencil=false
       fontMaterial=null
       fontObject=null
       letterSpacing=0
       monospacedWidth=1
       size=9
       skewAmount=0
       typefaceFontName=Regular
       outlineSettings {
        bApplyOutlineToDropShadows=false
        bMiteredCorners=false
        bSeparateFillAlpha=false
        outlineMaterial=null
        outlineSize=0
        outlineColor=<rgba 0 0 0 1>
       }
      }
      highlightColor {colorUseRule=UseColor_ColorTable, specifiedColor=<rgba 0 0 0 0>}
      highlightShape {
       bIsDynamicallyLoaded=false
       drawAs=Image
       imageType=NoImage
       mirroring=NoMirror
       resourceName=None
       resourceObject=null
       tiling=NoTile
       imageSize=<0 0>
       margin=<ltrb 0 0 0 0>
       outlineSettings {
        bUseBrushTransparency=false
        roundingType=HalfHeightRadius
        width=0
        color {colorUseRule=UseColor_Specified, specifiedColor=<rgba 0 0 0 0>}
        cornerRadii=<0 0 0 0>
       }
       tintColor {colorUseRule=UseColor_ColorTable, specifiedColor=<rgba 0 0 0 0>}
       uVRegion {bIsValid=false, max=<0 0>, min=<0 0>}
      }
      selectedBackgroundColor {colorUseRule=UseColor_ColorTable, specifiedColor=<rgba 0 0 0 0>}
      shadowColorAndOpacity=<rgba 0 0 0 1>
      shadowOffset=<0 0>
      strikeBrush {
       bIsDynamicallyLoaded=false
       drawAs=Image
       imageType=NoImage
       mirroring=NoMirror
       resourceName=None
       resourceObject=null
       tiling=NoTile
       imageSize=<32 32>
       margin=<ltrb 0 0 0 0>
       outlineSettings {
        bUseBrushTransparency=false
        roundingType=HalfHeightRadius
        width=0
        color {colorUseRule=UseColor_Specified, specifiedColor=<rgba 0 0 0 0>}
        cornerRadii=<0 0 0 0>
       }
       tintColor {colorUseRule=UseColor_Specified, specifiedColor=<rgba 1 1 1 1>}
       uVRegion {bIsValid=false, max=<0 0>, min=<0 0>}
      }
      underlineBrush {
       bIsDynamicallyLoaded=false
       drawAs=Image
       imageType=NoImage
       mirroring=NoMirror
       resourceName=None
       resourceObject=null
       tiling=NoTile
       imageSize=<32 32>
       margin=<ltrb 0 0 0 0>
       outlineSettings {
        bUseBrushTransparency=false
        roundingType=HalfHeightRadius
        width=0
        color {colorUseRule=UseColor_Specified, specifiedColor=<rgba 0 0 0 0>}
        cornerRadii=<0 0 0 0>
       }
       tintColor {colorUseRule=UseColor_Specified, specifiedColor=<rgba 1 1 1 1>}
       uVRegion {bIsValid=false, max=<0 0>, min=<0 0>}
      }
     }
     vScrollBarPadding=<ltrb 0 0 0 0>
    }
   }
   slot {class=HorizontalBoxSlot, properties {horizontalAlignment=HAlign_Center}}
  }
  btnPageRight {
   widgetClassPath=&TA_Generic01_Button
   parent=HorizontalBox_1
   widget=btnPageRight
   properties {label=" Next Page >>>", visibility=Visible, buttonColor_Default=<rgba 0 0.474 0.0147 1>, buttonColor_Selected=<rgba 0.6828 1 0.0254 1>}
   slot {class=HorizontalBoxSlot, properties {horizontalAlignment=HAlign_Center, verticalAlignment=VAlign_Center, padding=<ltrb 5 0 5 0>}}
  }
  PnlAssetControls {widgetClassPath=&HorizontalBox, bIsVariable=true, parent=VerticalBox_0, widget=PnlAssetControls, slot {class=VerticalBoxSlot}}
  VerticalBox_82 {
   widgetClassPath=&VerticalBox
   parent=PnlAssetControls
   widget=VerticalBox_82
   slot {class=HorizontalBoxSlot, properties {size {sizeRule=Fill, value=1}}}
  }
  VbExportToText {
   widgetClassPath=&VerticalBox
   bIsVariable=true
   parent=VerticalBox_82
   widget=VbExportToText
   properties {visibility=Collapsed}
   slot {class=VerticalBoxSlot}
  }
  PropOutputFolder {
   widgetClassPath=&SinglePropertyView
   parent=VbExportToText
   widget=PropOutputFolder
   properties {visibility=SelfHitTestInvisible}
   slot {class=VerticalBoxSlot, properties {size {sizeRule=Fill, value=1}}}
  }
  btnExportToText {
   widgetClassPath=&TA_Generic01_Button
   parent=VbExportToText
   widget=btnExportToText
   properties {label=" Export to text files "}
   slot {class=VerticalBoxSlot, properties {padding=<ltrb 5 0 5 0>}}
  }
  HorizontalBox_2 {widgetClassPath=&HorizontalBox, parent=VerticalBox_82, widget=HorizontalBox_2, slot {class=VerticalBoxSlot}}
  VbGenerateSizeReport {widgetClassPath=&VerticalBox, bIsVariable=true, parent=VerticalBox_82, widget=VbGenerateSizeReport, slot {class=VerticalBoxSlot}}
  HorizontalBox {widgetClassPath=&HorizontalBox, parent=VbGenerateSizeReport, widget=HorizontalBox, slot {class=VerticalBoxSlot}}
  PropEditorClassSizeOffset {
   widgetClassPath=&SinglePropertyView
   parent=VbGenerateSizeReport
   widget=PropEditorClassSizeOffset
   slot {class=VerticalBoxSlot}
  }
  btnGenerateSizeReport {
   widgetClassPath=&TA_Generic01_Button
   parent=VbGenerateSizeReport
   widget=btnGenerateSizeReport
   properties {label="Generate Class Size Report "}
   slot {class=VerticalBoxSlot, properties {padding=<ltrb 5 0 5 0>}}
  }
  VbGeneratePropertyReport {
   widgetClassPath=&VerticalBox
   bIsVariable=true
   parent=VerticalBox_82
   widget=VbGeneratePropertyReport
   slot {class=VerticalBoxSlot}
  }
  VerticalBox_149 {widgetClassPath=&VerticalBox, parent=VbGeneratePropertyReport, widget=VerticalBox_149, slot {class=VerticalBoxSlot}}
  HorizontalBox_94 {widgetClassPath=&HorizontalBox, parent=VerticalBox_149, widget=HorizontalBox_94, slot {class=VerticalBoxSlot}}
  PropCheckTimesDefaultChanges {
   widgetClassPath=&SinglePropertyView
   parent=HorizontalBox_94
   widget=PropCheckTimesDefaultChanges
   slot {class=HorizontalBoxSlot}
  }
  PropLogChangedProperties {
   widgetClassPath=&SinglePropertyView
   parent=HorizontalBox_94
   widget=PropLogChangedProperties
   slot {class=HorizontalBoxSlot, properties {size {sizeRule=Fill, value=1}}}
  }
  HorizontalBox_120 {widgetClassPath=&HorizontalBox, parent=VerticalBox_149, widget=HorizontalBox_120, slot {class=VerticalBoxSlot}}
  PropOnlyShowEditable {
   widgetClassPath=&SinglePropertyView
   parent=HorizontalBox_120
   widget=PropOnlyShowEditable
   slot {class=HorizontalBoxSlot}
  }
  PropShowInstanceEditable {
   widgetClassPath=&SinglePropertyView
   parent=HorizontalBox_120
   widget=PropShowInstanceEditable
   slot {class=HorizontalBoxSlot}
  }
  PropShowBPModifiable {
   widgetClassPath=&SinglePropertyView
   parent=HorizontalBox_120
   widget=PropShowBPModifiable
   slot {class=HorizontalBoxSlot}
  }
  PropShowExposedInSpawn {
   widgetClassPath=&SinglePropertyView
   parent=HorizontalBox_120
   widget=PropShowExposedInSpawn
   slot {class=HorizontalBoxSlot}
  }
  btnGeneratePropertyReport {
   widgetClassPath=&TA_Generic01_Button
   parent=VbGeneratePropertyReport
   widget=btnGeneratePropertyReport
   properties {label="Generate Class Property Report "}
   slot {class=VerticalBoxSlot, properties {padding=<ltrb 5 0 5 0>, size {sizeRule=Fill, value=1}}}
  }
  btnCancelExport {
   widgetClassPath=&TA_Generic01_Button
   parent=PnlAssetControls
   widget=btnCancelExport
   properties {label=Cancel, visibility=Collapsed, buttonColor_Default=<rgba 1 0 0.0508 1>, buttonColor_Selected=<rgba 1 0.3628 0 1>}
   slot {class=HorizontalBoxSlot, properties {verticalAlignment=VAlign_Center, padding=<ltrb 5 0 5 0>, size {sizeRule=Fill, value=1}}}
  }
  txtProgress {
   widgetClassPath=&EditableTextBox
   bIsVariable=true
   parent=PnlAssetControls
   widget=txtProgress
   properties {
    isReadOnly=true
    justification=Center
    minimumDesiredWidth=25.0596
    text=???
    visibility=Collapsed
    widgetStyle {
     backgroundColor {colorUseRule=UseColor_Specified, specifiedColor=<rgba 0.1 0.1 0.1 1>}
     backgroundImageFocused {
      bIsDynamicallyLoaded=false
      drawAs=RoundedBox
      imageType=NoImage
      mirroring=NoMirror
      resourceName=None
      resourceObject=null
      tiling=NoTile
      imageSize=<0 0>
      margin=<ltrb 0 0 0 0>
      outlineSettings {
       bUseBrushTransparency=false
       roundingType=FixedRadius
       width=1
       color {colorUseRule=UseColor_ColorTable, specifiedColor=<rgba 0 0 0 0>}
       cornerRadii=<4 4 4 4>
      }
      tintColor {colorUseRule=UseColor_Specified, specifiedColor=<rgba 0.0048 0.0048 0.0048 1>}
      uVRegion {bIsValid=false, max=<0 0>, min=<0 0>}
     }
     backgroundImageHovered {
      bIsDynamicallyLoaded=false
      drawAs=RoundedBox
      imageType=NoImage
      mirroring=NoMirror
      resourceName=None
      resourceObject=null
      tiling=NoTile
      imageSize=<0 0>
      margin=<ltrb 0 0 0 0>
      outlineSettings {
       bUseBrushTransparency=false
       roundingType=FixedRadius
       width=1
       color {colorUseRule=UseColor_ColorTable, specifiedColor=<rgba 0 0 0 0>}
       cornerRadii=<4 4 4 4>
      }
      tintColor {colorUseRule=UseColor_Specified, specifiedColor=<rgba 0.0048 0.0048 0.0048 1>}
      uVRegion {bIsValid=false, max=<0 0>, min=<0 0>}
     }
     backgroundImageNormal {
      bIsDynamicallyLoaded=false
      drawAs=RoundedBox
      imageType=NoImage
      mirroring=NoMirror
      resourceName=None
      resourceObject=null
      tiling=NoTile
      imageSize=<0 0>
      margin=<ltrb 0 0 0 0>
      outlineSettings {
       bUseBrushTransparency=false
       roundingType=FixedRadius
       width=1
       color {colorUseRule=UseColor_ColorTable, specifiedColor=<rgba 0 0 0 0>}
       cornerRadii=<4 4 4 4>
      }
      tintColor {colorUseRule=UseColor_Specified, specifiedColor=<rgba 0.0048 0.0048 0.0048 1>}
      uVRegion {bIsValid=false, max=<0 0>, min=<0 0>}
     }
     backgroundImageReadOnly {
      bIsDynamicallyLoaded=false
      drawAs=RoundedBox
      imageType=NoImage
      mirroring=NoMirror
      resourceName=None
      resourceObject=null
      tiling=NoTile
      imageSize=<0 0>
      margin=<ltrb 0 0 0 0>
      outlineSettings {
       bUseBrushTransparency=false
       roundingType=FixedRadius
       width=1
       color {colorUseRule=UseColor_ColorTable, specifiedColor=<rgba 0 0 0 0>}
       cornerRadii=<4 4 4 4>
      }
      tintColor {colorUseRule=UseColor_Specified, specifiedColor=<rgba 0.0284 0.0284 0.0284 1>}
      uVRegion {bIsValid=false, max=<0 0>, min=<0 0>}
     }
     focusedForegroundColor {colorUseRule=UseColor_Specified, specifiedColor=<rgba 1 1 1 1>}
     foregroundColor {colorUseRule=UseColor_Specified, specifiedColor=<rgba 1 1 1 1>}
     hScrollBarPadding=<ltrb 0 0 0 0>
     padding=<ltrb 8 3 8 4>
     readOnlyForegroundColor {colorUseRule=UseColor_Specified, specifiedColor=<rgba 1 1 1 1>}
     scrollBarStyle {
      thickness=8
      draggedThumbImage {
       bIsDynamicallyLoaded=false
       drawAs=RoundedBox
       imageType=NoImage
       mirroring=NoMirror
       resourceName=None
       resourceObject=null
       tiling=NoTile
       imageSize=<0 0>
       margin=<ltrb 0 0 0 0>
       outlineSettings {
        bUseBrushTransparency=false
        roundingType=FixedRadius
        width=0
        color {colorUseRule=UseColor_Specified, specifiedColor=<rgba 0 0 0 0>}
        cornerRadii=<4 4 4 4>
       }
       tintColor {colorUseRule=UseColor_Specified, specifiedColor=<rgba 0.2159 0.2159 0.2159 1>}
       uVRegion {bIsValid=false, max=<0 0>, min=<0 0>}
      }
      horizontalBackgroundImage {
       bIsDynamicallyLoaded=false
       drawAs=NoDrawType
       imageType=NoImage
       mirroring=NoMirror
       resourceName=None
       resourceObject=null
       tiling=NoTile
       imageSize=<0 0>
       margin=<ltrb 0 0 0 0>
       outlineSettings {
        bUseBrushTransparency=false
        roundingType=HalfHeightRadius
        width=0
        color {colorUseRule=UseColor_Specified, specifiedColor=<rgba 0 0 0 0>}
        cornerRadii=<0 0 0 0>
       }
       tintColor {colorUseRule=UseColor_Specified, specifiedColor=<rgba 1 1 1 1>}
       uVRegion {bIsValid=false, max=<0 0>, min=<0 0>}
      }
      horizontalBottomSlotImage {
       bIsDynamicallyLoaded=false
       drawAs=NoDrawType
       imageType=NoImage
       mirroring=NoMirror
       resourceName=None
       resourceObject=null
       tiling=NoTile
       imageSize=<0 0>
       margin=<ltrb 0 0 0 0>
       outlineSettings {
        bUseBrushTransparency=false
        roundingType=HalfHeightRadius
        width=0
        color {colorUseRule=UseColor_Specified, specifiedColor=<rgba 0 0 0 0>}
        cornerRadii=<0 0 0 0>
       }
       tintColor {colorUseRule=UseColor_Specified, specifiedColor=<rgba 1 1 1 1>}
       uVRegion {bIsValid=false, max=<0 0>, min=<0 0>}
      }
      horizontalTopSlotImage {
       bIsDynamicallyLoaded=false
       drawAs=NoDrawType
       imageType=NoImage
       mirroring=NoMirror
       resourceName=None
       resourceObject=null
       tiling=NoTile
       imageSize=<0 0>
       margin=<ltrb 0 0 0 0>
       outlineSettings {
        bUseBrushTransparency=false
        roundingType=HalfHeightRadius
        width=0
        color {colorUseRule=UseColor_Specified, specifiedColor=<rgba 0 0 0 0>}
        cornerRadii=<0 0 0 0>
       }
       tintColor {colorUseRule=UseColor_Specified, specifiedColor=<rgba 1 1 1 1>}
       uVRegion {bIsValid=false, max=<0 0>, min=<0 0>}
      }
      hoveredThumbImage {
       bIsDynamicallyLoaded=false
       drawAs=RoundedBox
       imageType=NoImage
       mirroring=NoMirror
       resourceName=None
       resourceObject=null
       tiling=NoTile
       imageSize=<0 0>
       margin=<ltrb 0 0 0 0>
       outlineSettings {
        bUseBrushTransparency=false
        roundingType=FixedRadius
        width=0
        color {colorUseRule=UseColor_Specified, specifiedColor=<rgba 0 0 0 0>}
        cornerRadii=<4 4 4 4>
       }
       tintColor {colorUseRule=UseColor_Specified, specifiedColor=<rgba 0.2159 0.2159 0.2159 1>}
       uVRegion {bIsValid=false, max=<0 0>, min=<0 0>}
      }
      normalThumbImage {
       bIsDynamicallyLoaded=false
       drawAs=RoundedBox
       imageType=NoImage
       mirroring=NoMirror
       resourceName=None
       resourceObject=null
       tiling=NoTile
       imageSize=<0 0>
       margin=<ltrb 0 0 0 0>
       outlineSettings {
        bUseBrushTransparency=false
        roundingType=FixedRadius
        width=0
        color {colorUseRule=UseColor_Specified, specifiedColor=<rgba 0 0 0 0>}
        cornerRadii=<4 4 4 4>
       }
       tintColor {colorUseRule=UseColor_Specified, specifiedColor=<rgba 0.0953 0.0953 0.0953 1>}
       uVRegion {bIsValid=false, max=<0 0>, min=<0 0>}
      }
      verticalBackgroundImage {
       bIsDynamicallyLoaded=false
       drawAs=NoDrawType
       imageType=NoImage
       mirroring=NoMirror
       resourceName=None
       resourceObject=null
       tiling=NoTile
       imageSize=<0 0>
       margin=<ltrb 0 0 0 0>
       outlineSettings {
        bUseBrushTransparency=false
        roundingType=HalfHeightRadius
        width=0
        color {colorUseRule=UseColor_Specified, specifiedColor=<rgba 0 0 0 0>}
        cornerRadii=<0 0 0 0>
       }
       tintColor {colorUseRule=UseColor_Specified, specifiedColor=<rgba 1 1 1 1>}
       uVRegion {bIsValid=false, max=<0 0>, min=<0 0>}
      }
      verticalBottomSlotImage {
       bIsDynamicallyLoaded=false
       drawAs=NoDrawType
       imageType=NoImage
       mirroring=NoMirror
       resourceName=None
       resourceObject=null
       tiling=NoTile
       imageSize=<0 0>
       margin=<ltrb 0 0 0 0>
       outlineSettings {
        bUseBrushTransparency=false
        roundingType=HalfHeightRadius
        width=0
        color {colorUseRule=UseColor_Specified, specifiedColor=<rgba 0 0 0 0>}
        cornerRadii=<0 0 0 0>
       }
       tintColor {colorUseRule=UseColor_Specified, specifiedColor=<rgba 1 1 1 1>}
       uVRegion {bIsValid=false, max=<0 0>, min=<0 0>}
      }
      verticalTopSlotImage {
       bIsDynamicallyLoaded=false
       drawAs=NoDrawType
       imageType=NoImage
       mirroring=NoMirror
       resourceName=None
       resourceObject=null
       tiling=NoTile
       imageSize=<0 0>
       margin=<ltrb 0 0 0 0>
       outlineSettings {
        bUseBrushTransparency=false
        roundingType=HalfHeightRadius
        width=0
        color {colorUseRule=UseColor_Specified, specifiedColor=<rgba 0 0 0 0>}
        cornerRadii=<0 0 0 0>
       }
       tintColor {colorUseRule=UseColor_Specified, specifiedColor=<rgba 1 1 1 1>}
       uVRegion {bIsValid=false, max=<0 0>, min=<0 0>}
      }
     }
     textStyle {
      overflowPolicy=Clip
      transformPolicy=None
      colorAndOpacity {colorUseRule=UseColor_Foreground, specifiedColor=<rgba 1 0 1 1>}
      font {
       bForceMonospaced=false
       bMaterialIsStencil=false
       fontMaterial=null
       fontObject=null
       letterSpacing=0
       monospacedWidth=1
       size=9
       skewAmount=0
       typefaceFontName=Regular
       outlineSettings {
        bApplyOutlineToDropShadows=false
        bMiteredCorners=false
        bSeparateFillAlpha=false
        outlineMaterial=null
        outlineSize=0
        outlineColor=<rgba 0 0 0 1>
       }
      }
      highlightColor {colorUseRule=UseColor_ColorTable, specifiedColor=<rgba 0 0 0 0>}
      highlightShape {
       bIsDynamicallyLoaded=false
       drawAs=Image
       imageType=NoImage
       mirroring=NoMirror
       resourceName=None
       resourceObject=null
       tiling=NoTile
       imageSize=<0 0>
       margin=<ltrb 0 0 0 0>
       outlineSettings {
        bUseBrushTransparency=false
        roundingType=HalfHeightRadius
        width=0
        color {colorUseRule=UseColor_Specified, specifiedColor=<rgba 0 0 0 0>}
        cornerRadii=<0 0 0 0>
       }
       tintColor {colorUseRule=UseColor_ColorTable, specifiedColor=<rgba 0 0 0 0>}
       uVRegion {bIsValid=false, max=<0 0>, min=<0 0>}
      }
      selectedBackgroundColor {colorUseRule=UseColor_ColorTable, specifiedColor=<rgba 0 0 0 0>}
      shadowColorAndOpacity=<rgba 0 0 0 1>
      shadowOffset=<0 0>
      strikeBrush {
       bIsDynamicallyLoaded=false
       drawAs=Image
       imageType=NoImage
       mirroring=NoMirror
       resourceName=None
       resourceObject=null
       tiling=NoTile
       imageSize=<32 32>
       margin=<ltrb 0 0 0 0>
       outlineSettings {
        bUseBrushTransparency=false
        roundingType=HalfHeightRadius
        width=0
        color {colorUseRule=UseColor_Specified, specifiedColor=<rgba 0 0 0 0>}
        cornerRadii=<0 0 0 0>
       }
       tintColor {colorUseRule=UseColor_Specified, specifiedColor=<rgba 1 1 1 1>}
       uVRegion {bIsValid=false, max=<0 0>, min=<0 0>}
      }
      underlineBrush {
       bIsDynamicallyLoaded=false
       drawAs=Image
       imageType=NoImage
       mirroring=NoMirror
       resourceName=None
       resourceObject=null
       tiling=NoTile
       imageSize=<32 32>
       margin=<ltrb 0 0 0 0>
       outlineSettings {
        bUseBrushTransparency=false
        roundingType=HalfHeightRadius
        width=0
        color {colorUseRule=UseColor_Specified, specifiedColor=<rgba 0 0 0 0>}
        cornerRadii=<0 0 0 0>
       }
       tintColor {colorUseRule=UseColor_Specified, specifiedColor=<rgba 1 1 1 1>}
       uVRegion {bIsValid=false, max=<0 0>, min=<0 0>}
      }
     }
     vScrollBarPadding=<ltrb 0 0 0 0>
    }
   }
   slot {class=HorizontalBoxSlot}
  }
 ]
}
